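# Masked, lit Niagara sprite material for the dissolving particles of the Shaded Expanse aura cosmetic in SparksCosmetics.
# Blended cell-noise and sphere normals drive a fake directional light and reflection-texture highlight, tinted by particle color, to build emissive.
# Shape combines a tapered, wind-distorted sphere mask with two panning, per-particle randomized cell-noise layers.
# A further polar-coordinate noise is thresholded with a step, apparently for the hard-edged dissolve.
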Material /SparksCosmetics/Items/Auras/Assets/NS_Aura_ShadedExpanse/M_Aura_ShadedExpanse_Dissolve
graph {
 properties {bAutomaticallySetUsageInEditor=false, bUsedWithNiagaraSprites=true, blendMode=BLEND_Masked, usedShadingModels=Default Lit}
 nodes [
  n1 = TextureSample(UVs=NamedRerouteUsage()) {samplerType=SAMPLERTYPE_Normal, texture=/Game/Effects/Fort_Effects/Textures/Tile/T_CellNoise_Tile_N}
  n2 = Constant() {r=0.3}
  n3 = TextureSample(UVs=NamedRerouteUsage()) {samplerType=SAMPLERTYPE_Normal, texture=/Game/Effects/Fort_Effects/Textures/Mask/T_Sphere_N}
  n4 = Constant() {r=0.5}
  NamedRerouteDeclaration(None=MaterialFunctionCall(
    BaseNormal=MaterialFunctionCall(Normal=@n1.RGB, Flatness=@n2) {materialFunction=&FlattenNormal}.Result
    AdditionalNormal=MaterialFunctionCall(Normal=@n3.RGB, Flatness=@n4) {materialFunction=&FlattenNormal}.Result
   ) {materialFunction=&BlendAngleCorrectedNormals}.Result) {name=Normals, nodeColor=<rgba 0 0.1797 1 1>}
  n5 = ParticleColor()
  NamedRerouteDeclaration(None=@n5.RGB) {name=Particle RGB, nodeColor=<rgba 1 0.3516 0 1>}
  n6 = ReflectionVectorWS(CustomWorldNormal=Transform(None=NamedRerouteUsage()))
  n7 = Transform(None=VectorParameter() {group=DirectionalLight, parameterName=Fake Light Direction, defaultValue=<rgba 0.5 0.5 1 0>}.RGB)
  n8 = Power(
   Base=ConstantBiasScale(None=DotProduct(A=@n6, B=@n7))
   Exp=ScalarParameter() {defaultValue=6, group=DirectionalLight, parameterName=DirectionalLightPower}
  ) {constExponent=6}
  n9 = ScalarParameter() {defaultValue=1, group=DirectionalLight, parameterName=DirectionalLightStrength}
  n10 = Multiply(A=Saturate(None=Multiply(A=@n8, B=@n9) {constB=3}), B=NamedRerouteUsage())
  NamedRerouteDeclaration(None=@n10) {name=Base Color, nodeColor=<rgba 0 0.4844 1 1>}
  NamedRerouteDeclaration(None=@n5.A) {name=Particle Alpha, nodeColor=<rgba 0.0312 0 1 1>}
  n11 = LinearInterpolate(B=@n10, Alpha=ScalarParameter() {group=DirectionalLight, parameterName=AddSoftDirectionalLightEmissive})
  n12 = ComponentMask(None=@n6) {g=true, r=true}
  n13 = Multiply(
   A=TextureSampleParameter2D(UVs=Multiply(A=@n12)) {parameterName=ReflectionTexture, texture=/Game/Packages/Fortress_SharedMaterials/Textures/T_Ravens_SpecularReflection}.R
   B=ScalarParameter() {defaultValue=0.5, group=DirectionalLight, parameterName=HighlightStrength}
  ) {constB=0.5}
  n14 = ScalarParameter() {defaultValue=0.2, group=DirectionalLight, parameterName=HighlightContrast}
  n15 = LinearInterpolate(
   B=Saturate(None=MaterialFunctionCall!(In=@n13, Contrast=@n14) {materialFunction=/Engine/Functions/Engine_MaterialFunctions01/ImageAdjustment/CheapContrast_RGB}.Result)
   Alpha=ScalarParameter() {defaultValue=1, group=DirectionalLight, parameterName=AddSpecularHighlight}
  )
  NamedRerouteDeclaration(None=MaterialFunctionCall(Emissive=Add(A=@n11, B=@n15)) {materialFunction=&MF_FX_EmissiveAdjustments}.Result) {name=Emissive, nodeColor=<rgba 1 0 0.6562 1>}
  n16 = Add(A=Panner(Coordinate=TextureCoordinate()) {speedY=0.4}, B=ParticleRandom())
  NamedRerouteDeclaration(None=Multiply(
    A=TextureSample(UVs=@n16) {samplerType=SAMPLERTYPE_Normal, texture=/Game/Effects/Fort_Effects/Textures/Tile/T_Wind_N}.R
    B=ScalarParameter() {defaultValue=0.1, group=UVControls, parameterName=UV Distortion Strength}
   ) {constB=0.08}) {name=UV Distortion, nodeColor=<rgba 0.9688 0 1 1>}
  n17 = TextureCoordinate()
  n18 = Constant2Vector() {g=1, r=1}
  n19 = LinearInterpolate(Alpha=Power(Base=MaterialFunctionCall() {materialFunction=&LinearGradient}.VGradient)) {constA=1.1, constB=0.3}
  n20 = Constant() {r=1}
  n21 = Add(
   A=MaterialFunctionCall(UVs=Multiply(A=@n17, B=@n18), Texture Scale=AppendVector(A=@n19, B=@n20)) {materialFunction=&ScaleUVsByCenter}.UVs
   B=NamedRerouteUsage()
  )
  NamedRerouteDeclaration(None=@n21) {name=MaskShape UVS, nodeColor=<rgba 0.0781 0 1 1>}
  n22 = NamedRerouteUsage()
  n23 = Multiply(A=TextureCoordinate(), B=ScalarParameter() {defaultValue=3.5, group=UVControls, parameterName=UV Scale Base Texture}) {constB=3.5}
  n24 = Add(A=Add(A=Panner(Coordinate=@n23) {speedX=-0.1, speedY=0.5}, B=NamedRerouteUsage()), B=ParticleRandom())
  n25 = TextureSampleParameter2D(UVs=@n24) {
   group=Texture
   parameterName=Main Texture
   samplerType=SAMPLERTYPE_LinearGrayscale
   texture=/Game/Effects/Fort_Effects/Textures/Tile/T_CellNoise_Tile_M
  }
  n26 = Saturate(None=Add(A=NamedRerouteUsage(), B=@n25.R))
  n27 = ComponentMask(None=TextureCoordinate()) {g=true}
  n28 = Multiply(A=TextureCoordinate(), B=ScalarParameter() {defaultValue=4.5, group=UVControls, parameterName=UV Scale Detail Texture}) {constB=5}
  n29 = TextureSampleParameter2D(UVs=Add(A=Panner(Coordinate=@n28) {speedX=0.1, speedY=0.55}, B=ParticleRandom())) {
   group=Texture
   parameterName=DetailTexture
   samplerType=SAMPLERTYPE_LinearGrayscale
   texture=&T_CellNoise_Tile_M
  }
  n30 = Max(A=Saturate(None=Multiply(A=@n27) {constB=2}), B=@n29.G)
  n31 = Add(A=Saturate(None=Power(Base=@n22) {constExponent=7}), B=Multiply(A=NamedRerouteUsage(), B=Multiply(A=@n26, B=@n30)) {constA=1})
  NamedRerouteDeclaration(None=Multiply(A=NamedRerouteUsage(), B=Saturate(None=@n31))) {name=Opacity, nodeColor=<rgba 0.7109 0 1 1>}
  NamedRerouteDeclaration(None=@n24) {name=UVs Base Texture, nodeColor=<rgba 0.1719 1 0 1>}
  NamedRerouteDeclaration(None=SphereMask(A=@n21, B=Constant() {r=0.5}, Radius=ScalarParameter() {defaultValue=0.45, group=Mask, parameterName=SphereMask Radius}) {attenuationRadius=0.47, hardnessPercent=0}) {name=SphereMask, nodeColor=<rgba 0.9453 0 1 1>}
  NamedRerouteDeclaration(None=Multiply(A=Power(Base=@n27) {constExponent=5}, B=OneMinus(None=@n27))) {name=UV Mask, nodeColor=<rgba 1 0 0.2578 1>}
  NamedRerouteUsage()
  NamedRerouteUsage()
  NamedRerouteUsage()
  NamedRerouteUsage()
  ScalarParameter() {defaultValue=0.15, parameterName=Specular}
  ScalarParameter() {defaultValue=0.1, parameterName=Roughness}
  SmoothStep(
   Min=ScalarParameter() {defaultValue=0.2, group=Mask, parameterName=SmoothStepMin}
   Max=ScalarParameter() {defaultValue=0.6, group=Mask, parameterName=SmoothStepMax}
   Value=NamedRerouteUsage()
  ) {constMax=0.6, constMin=0.2}
  n32 = NamedRerouteUsage()
  n33 = Constant()
  n34 = Constant() {r=-0.5}
  MaterialExpressionMaterialFunctionCall_0 = MaterialFunctionCall!(UV Coord=Add(A=TextureCoordinate(), B=AppendVector(A=@n33, B=@n34)) {constB=-0.5}, LenghtScale, RadialScale) {materialFunction=/Game/Global/MaterialFunctions/TechArt/MF_PolarCoordinates}
  n35 = Convert(None=@MaterialExpressionMaterialFunctionCall_0.Result) {
   nodeName=Break Float2
   convertInputs [{
     type=Vector2
     expressionInput {
      expression=&self:MaterialExpressionMaterialFunctionCall_0
      inputName=None
      mask=0
      maskA=0
      maskB=0
      maskG=0
      maskR=0
      outputIndex=0
     }
    }]
   convertOutputs [{type=Scalar}, {type=Scalar}]
  }
  n36 = Power(Base=@n35) {constExponent=4}
  n37 = Constant() {r=0.5}
  n38 = AppendVector(A=@n35, B=Add(A=@n35, B=Multiply(A=@n36, B=@n37)))
  n39 = TextureSample(UVs=Multiply(A=Panner(Coordinate=@n38) {speedX=-0.2}) {constB=3}) {samplerType=SAMPLERTYPE_LinearGrayscale, texture=&T_CellNoise_Tile_M}
  Step(Y=Multiply(A=NamedRerouteUsage(), B=Add(A=@n32, B=@n39.R))) {constX=0.4}
 ]
}
