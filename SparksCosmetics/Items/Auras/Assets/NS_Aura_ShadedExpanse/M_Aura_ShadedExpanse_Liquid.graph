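# Lit liquid material for Niagara mesh particles in the Shaded Expanse aura.
# Reflection-vector and panning screen-aligned UVs sample a rainbow ramp, desaturated by parameter and masked by Fresnel for base color.
# A Fresnel rim tinted by particle color feeds emissive, and a panning normal map is flattened for surface detail.
# Panning curl noise in pivot-relative space drives world position offset, scaled by a particle dynamic parameter and zeroed on mobile feature levels.
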Material /SparksCosmetics/Items/Auras/Assets/NS_Aura_ShadedExpanse/M_Aura_ShadedExpanse_Liquid
graph {
 properties {bAutomaticallySetUsageInEditor=false, bUsedWithNiagaraMeshParticles=true, usedShadingModels=Default Lit}
 nodes [
  n1 = ReflectionVectorWS()
  n2 = ScalarParameter() {defaultValue=0.2, parameterName=Reflection Scale}
  n3 = MaterialFunctionCall!(Center Point World Position) {materialFunction=/Engine/Functions/Engine_MaterialFunctions02/UVs/WorldSpaceAlignedScreenCoordinates}
  n4 = Constant() {r=6}
  n5 = AppendVector(A=@n4, B=@n4)
  n6 = Add(
   A=MaterialFunctionCall!(
    UV=ComponentMask(None=@n1) {r=true}
    Tile U=@n2
    Tile V=@n2
    Pan U
    Pan V
    Offset U=ScalarParameter() {defaultValue=0.5, parameterName=ReflectionOffset}
    Offset V
    Time
   ) {materialFunction=/Game/Effects/Fort_Effects/Materials/_Functions/MF_PanTileOffset}.UV's
   B=Panner(Coordinate=Multiply(A=@n3.Offset and Scaled UVS, B=@n5)) {speedX=0.1, speedY=-0.05}
  )
  n7 = TextureSample(UVs=@n6) {texture=/Game/Effects/Fort_Effects/Textures/Tile/T_Inky_Smoke_Tile}
  Add(A=@n6, B=Multiply(A=@n7.R) {constB=0.5})
  n8 = Constant()
  AppendVector(A=@n7.R, B=@n8)
  Constant() {r=0.5}
  Constant() {r=0.1}
  Constant() {r=0.2}
  Constant()
  MaterialFunctionCall(Normal=NamedRerouteUsage(), Flatness=ScalarParameter() {defaultValue=0.5, parameterName=NormalStrength}) {materialFunction=&FlattenNormal}
  n9 = Fresnel(ExponentIn=Constant() {r=2.5})
  n10 = Power(Base=@n9) {constExponent=3}
  n11 = ParticleColor()
  n12 = NamedRerouteUsage()
  n13 = ScalarParameter() {defaultValue=0.4, parameterName=Emissive Mult}
  MaterialFunctionCall(Emissive=Add(A=Multiply(A=@n10, B=@n11.RGB), B=Multiply(A=@n12, B=@n13))) {materialFunction=&MF_FX_EmissiveAdjustments}
  Multiply(A=VertexNormalWS(), B=Constant() {r=-4})
  n14 = Multiply(A=TextureCoordinate(), B=ScalarParameter() {defaultValue=0.25, parameterName=Normal Scale})
  n15 = AppendVector(A=Multiply(A=ScalarParameter() {defaultValue=0.01, parameterName=Normal Pan U}, B=Time()), B=Constant())
  NamedRerouteDeclaration(None=TextureSample(UVs=Add(A=@n14, B=@n15)) {samplerType=SAMPLERTYPE_Normal, texture=/Game/Global/Textures/Default/Tileable_Patterns/2ndInfectedMapNormal}.RGB) {name=NormalMap, nodeColor=<rgba 1 0.8438 0 1>}
  n16 = WorldPosition()
  n17 = LinearInterpolate(
   A=@n16.XYZ
   B=Subtract(
    A=@n16.XYZ
    B=MaterialFunctionCall() {materialFunction=&ComponentPivotLocation}.Result
   )
   Alpha=ScalarParameter() {defaultValue=1, desc="1 = local, 0 = world space", parameterName=WPO relative to object Location y/n}
  )
  n18 = Multiply(A=Time(), B=ScalarParameter() {defaultValue=-15, parameterName=WPO Pan Speed})
  n19 = TextureSample(UVs=Divide(A=Add(A=@n17, B=@n18), B=ScalarParameter() {defaultValue=1600, parameterName=WPO Scale})) {samplerType=SAMPLERTYPE_LinearColor, texture=/Game/Effects/FogVolume/VT_CurlNoise_LOW}
  n20 = DynamicParameter() {defaultValue=<rgba 4 1 1 1>, paramNames [WPOStrength,Param2,Param3,Param4]}
  n21 = Constant()
  NamedRerouteDeclaration(None=FeatureLevelSwitch(Default=Multiply(A=@n19.RGB, B=@n20.WPOStrength), ES2=@n21, ES3_1=@n21)) {name=WPO, nodeColor=<rgba 1 0.1172 0 1>}
  n22 = AppendVector(A=Constant() {r=1}, B=@n8)
  n23 = Desaturation(
   None=TextureSampleParameter2D(UVs=Multiply(A=@n22, B=@n6)) {parameterName=ColorRemap, texture=/Game/Global/Textures/Default/Masks_and_Gradients/FortniteRainbow}.RGB
   Fraction=ScalarParameter() {parameterName=Desturate}
  )
  n24 = ScalarParameter() {defaultValue=0.6, parameterName=ColorMult}
  NamedRerouteDeclaration(None=LinearInterpolate(B=Multiply(A=@n23, B=@n24), Alpha=@n9)) {name=BaseColor, nodeColor=<rgba 0 0.8359 1 1>}
  NamedRerouteUsage()
  NamedRerouteUsage()
 ]
}
