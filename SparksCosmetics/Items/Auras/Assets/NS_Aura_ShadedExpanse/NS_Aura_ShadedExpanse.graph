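# Niagara system for the Shaded Expanse cosmetic aura, drawing animated ribbons and glows around a character's skeleton.
# A ribbon emitter partitions particles into two bone-anchored strands, shaped by curves and waveforms along camera axes, with velocity-scaled position smoothing and view-angle-dependent offsets.
# Decal_GlowRays pins four spawn groups of camera-offset sprites to selected bones.
# Large-body and front-end preview user flags adjust sizing, offsets and brightness.
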NiagaraSystem /SparksCosmetics/Items/Auras/Assets/NS_Aura_ShadedExpanse/NS_Aura_ShadedExpanse
dependencies {
 usedDataInterfaces [
  &NiagaraDataInterfaceArrayFloat
  &NiagaraDataInterfaceArrayInt32
  &NiagaraDataInterfaceArrayFloat3
  &NiagaraDataInterfaceCurve
  &NiagaraDataInterfaceCamera
  /Script/Niagara.NiagaraDataInterfaceArrayDistributionInt
  &NiagaraDataInterfaceVectorCurve
  &NiagaraDataInterfaceColorCurve
  &NiagaraDataInterfaceSkeletalMesh
 ]
 usedDynamicInputs [&MakeFloatFromVector2d]
 usedModules [
  &self:SystemSpawnScript.SystemScriptSource.SystemScriptGraph.NiagaraNodeAssignment_1.SetVariables_#1
  &SystemState
  &Timeline
  &self:SystemSpawnScript.SystemScriptSource.SystemScriptGraph.NiagaraNodeAssignment_2.SetVariables_#2
  &BlendSpaceHelper
  &self:SystemSpawnScript.SystemScriptSource.SystemScriptGraph.NiagaraNodeAssignment_0.SetVariables_#3
 ]
}
emitters {
 Camera_Oriented_Ribbon001 {
  emitterSpawnScript [/SetVariables(
    Emitter.NumberOfRibbons=2
    Emitter.ParticlesPerRibbon:NiagaraInt32=&MakeCustomIntFromBool(Bool=$User.IsLargeBody, True INT=20, False INT=20)
   ) {isSetParameters=true, module=SetVariables_#4}]
  emitterUpdateScript [
   &EmitterState()
   &SpawnBurst_Instantaneous(
    Spawn Count=&Multiply_Int(A=$Emitter.NumberOfRibbons, B=$Emitter.ParticlesPerRibbon)
    Use Spawn Probability=false
    UseLoopCountLimit=true
   )
   &SpawnRate(Use Spawn Probability=false) {enabled=false}
  ]
  particleSpawnScript [
   &PartitionParticles(
    Normalized Index In Partition Includes One=true
    Number Of Partitions=$Emitter.NumberOfRibbons
    Particles Per Partition:NiagaraInt32=$Emitter.ParticlesPerRibbon
    Partition Mode:ENiagara_PartitionMode_v2=::NewEnumerator1
    Write Partition to Ribbon ID=::NewEnumerator0
   )
   &InitializeParticle(
    Lifetime=5
    Ribbon Width=8
    Ribbon Width Mode=::NewEnumerator1
    Sprite Size Mode=::NewEnumerator3
    UsePositionOffset=false
    Write Material Random=true
   )
  ]
  particleUpdateScript [
   &ParticleState(Kill Particles When Lifetime Has Elapsed=false, Let Infinitely Lived Particles Die When Emitter Deactivates=true)
   /SetVariables(
    Particles.CurveAdjust=1.2
    Particles.HeightAdjust:NiagaraFloat=&Multiply_Float(
     A=&MakeCustomFloatFromBool(Bool=$User.IsLargeBody, True Float=1.25, False Float=1.1)
     B=&MakeCustomFloatFromBool(Bool=$User.bIsFrontEndPreview, True Float=1.1, False Float=1)
    )
    Particles.Waveform1:NiagaraFloat=&Waveform("Waveform [1]"=::NewEnumerator0, "[1] Phase Scale"=0.1, "[1] Frequency"=1, "[1] Amplitude Min/Max"=<0 1>)
    Particles.Waveform2:NiagaraFloat=&Waveform(
     "Waveform [1]"=::NewEnumerator0
     "[1] Phase Scale"=2.5
     "[1] Phase"=&Add_Float(A=$Particles.PartitionParticles.Normalized Index In Partition, B=&Multiply_Float(A=$Engine.Time, B=0.15))
     "[1] Frequency"=1
     "[1] Amplitude Min/Max"=<0 1>
     "Waveform [2]"=::NewEnumerator12
    )
    Particles.WidthAdjust:NiagaraFloat=&MakeCustomFloatFromBool(Bool=$User.IsLargeBody, True Float=1.25, False Float=1.1)
   ) {isSetParameters=true, module=SetVariables_#5}
   &SkeletalMeshLocation(
    Bone / Socket Index=7
    Bone Sampling Mode=::NewEnumerator4
    MaskSpawnGroup=false
    Orientation Sampling=::NewEnumerator1
    Sampled Position Offset=&MergeHelper_VectorSwitch(
     0=&AddVectorToPosition(
      Position=<-15 0 0>
      Vector=&MakeVector(
       X=&Add_Float(
        A=&FloatFromCurve(
         FloatCurve=DataInterface() {propertyValues {Curve {
            defaultValue=0
            keys [{interpMode=RCIM_Cubic, tangentMode=RCTM_User, time=0, value=1}, {interpMode=RCIM_Cubic, tangentMode=RCTM_User, time=1, value=0}]
           }}}
         CurveIndex=$Particles.PartitionParticles.Normalized Index In Partition
         Scale Curve=5
        )
        B=&Multiply_Float(
         A=&FloatFromCurve(
          FloatCurve=DataInterface() {propertyValues {Curve {
             defaultValue=0
             keys [
              {arriveTangent=4.0936, interpMode=RCIM_Cubic, leaveTangent=4.0936, tangentMode=RCTM_User, time=0, value=0}
              {interpMode=RCIM_Cubic, tangentMode=RCTM_User, time=0.5, value=1}
              {arriveTangent=-3.0128, tangentMode=RCTM_User, time=1, value=0}
             ]
            }}}
          CurveIndex=$Particles.PartitionParticles.Normalized Index In Partition
          Scale Curve=&Lerp_Float(A=32, B=40, Alpha=$Particles.Waveform1)
         )
         B=$Particles.HeightAdjust
        )
       )
       Y=0
       Z=0
      )
     )
     1=&AddVectorToPosition(
      Position=<-15 0 0>
      Vector=&MakeVector(
       X=&Add_Float(
        A=&FloatFromCurve(
         FloatCurve=DataInterface() {propertyValues {Curve {
            defaultValue=0
            keys [{interpMode=RCIM_Cubic, tangentMode=RCTM_User, time=0, value=1}, {interpMode=RCIM_Cubic, tangentMode=RCTM_User, time=1, value=0}]
           }}}
         CurveIndex=$Particles.PartitionParticles.Normalized Index In Partition
         Scale Curve=5
        )
        B=&FloatFromCurve(
         FloatCurve=DataInterface() {propertyValues {Curve {
            defaultValue=0
            keys [
             {arriveTangent=4.0936, interpMode=RCIM_Cubic, leaveTangent=4.0936, tangentMode=RCTM_User, time=0, value=0}
             {interpMode=RCIM_Cubic, tangentMode=RCTM_User, time=0.5, value=1}
             {arriveTangent=-3.0128, tangentMode=RCTM_User, time=1, value=0}
            ]
           }}}
         CurveIndex=$Particles.PartitionParticles.Normalized Index In Partition
         Scale Curve=&Lerp_Float(A=42, B=30, Alpha=$Particles.Waveform1)
        )
       )
       Y=0
       Z=0
      )
     )
     2=?
     3=?
     Index=$Particles.PartitionParticles.Partition
     Num=::NewEnumerator0
    )
    Skeletal Mesh=$System.Skeletal Mesh
    Use Sampled Mesh Attribute as Ribbon ID=false
   )
   &ScaleColor(
    Scale RGB=&Multiply_Vector(
     A=&MakeVectorFromLinearColorRGB(Color=&ColorFromCurve(
       ColorCurve=DataInterface() {propertyValues {
         AlphaCurve {keys [<at 0 1 RCIM_Cubic>]}
         BlueCurve {keys [<at 0.0229 0.0212 RCIM_Cubic>,<at 0.1897 0.2747>,<at 0.4637 0.2747>,<at 0.5356 0.2747>,<at 0.7777 0.2747>,<at 0.9715 0.0212 RCIM_Cubic>]}
         GreenCurve {keys [<at 0.0229 0.0529 RCIM_Cubic>,<at 0.1897 0.0438>,<at 0.4637 0>,<at 0.5356 0>,<at 0.7777 0.0595>,<at 0.9715 0.0529 RCIM_Cubic>]}
         RedCurve {keys [<at 0.0229 0 RCIM_Cubic>,<at 0.1897 0>,<at 0.4637 0.117>,<at 0.5356 0.117>,<at 0.7777 0>,<at 0.9715 0 RCIM_Cubic>]}
        }}
       CurveIndex=&Frac_Float(Float=&Add_Float(A=$Particles.PartitionParticles.Normalized Index In Partition, B=&Multiply_Float(A=$Engine.Time, B=0.5)))
      ))
     B=&VectorFromFloat(Value=0.3)
    )
    ScaleA=true
    ScaleRGB=true
   )
   &ScaleRibbonWidth(Ribbon Width Scale=&Multiply_Float(
     A=&MergeHelper_FloatSwitch(
      0=&Lerp_Float(A=0.1, B=1.25, Alpha=$Particles.Waveform2)
      1=&Lerp_Float(A=1.25, B=0.1, Alpha=$Particles.Waveform2)
      2=?
      3=?
      Index=$Particles.PartitionParticles.Partition
      Num=::NewEnumerator0
     )
     B=&FloatFromCurve(
      FloatCurve=DataInterface() {propertyValues {Curve {keys [
          {arriveTangent=4.1029, interpMode=RCIM_Cubic, leaveTangent=4.1029, tangentMode=RCTM_User, time=0, value=0.5}
          {arriveTangentWeight=0.2426, interpMode=RCIM_Cubic, leaveTangentWeight=0.082, tangentMode=RCTM_User, time=0.15, value=1}
          {arriveTangentWeight=0.1132, interpMode=RCIM_Cubic, leaveTangentWeight=0.2253, tangentMode=RCTM_User, time=0.2007, value=1}
          {arriveTangent=0.1507, interpMode=RCIM_Cubic, leaveTangent=0.675, tangentMode=RCTM_Break, time=0.5007, value=0.5}
          {arriveTangentWeight=0.2258, interpMode=RCIM_Cubic, leaveTangentWeight=0.1124, tangentMode=RCTM_User, time=0.8007, value=1}
          {arriveTangentWeight=0.135, interpMode=RCIM_Cubic, leaveTangentWeight=0.0442, tangentMode=RCTM_User, time=0.85, value=1}
          {arriveTangent=-8.6495, interpMode=RCIM_Cubic, leaveTangent=-8.6495, tangentMode=RCTM_User, time=0.9989, value=0.5}
         ]}}}
      CurveIndex=$Particles.PartitionParticles.Normalized Index In Partition
     )
    ))
   /SetVariables(Particles.VectortoCamera:Vector3f=&ReturnCameraVector(Camera Vectors=::NewEnumerator4, Coordinate Space=::Simulation, Origin Position=$Particles.Position)) {isSetParameters=true, module=SetVariables_#6}
   /SetVariables(Particles.CameraOffsetRibbon:NiagaraFloat=&Lerp_Float(
     A=&MakeCustomFloatFromBool(Bool=$User.bIsFrontEndPreview, True Float=-30, False Float=-35)
     B=-15
     Alpha=&MakeCustomFloatFromBool(
      Bool=&SetBoolByFloatComparison(
       A=&OneMinusFloat(Float=&DotProduct(Vector A=$Particles.VectortoCamera, Vector B=<0 0 1>))
       B=0.5
      )
      True Float=&FloatFromCurve(
       FloatCurve=DataInterface() {propertyValues {Curve {keys [<at 0.55 0>,<at 0.95 1>]}}}
       CurveIndex=&Abs_Float(Float:NiagaraFloat=&DotProduct(Vector A=$Particles.VectortoCamera, Vector B=<0 0 1>))
      )
      False Float=0
     )
    )) {isSetParameters=true, module=SetVariables_#7}
   &InterpolateOverTime(
    Rate Of Change:NiagaraFloat=&FloatFromCurve(
     FloatCurve=DataInterface() {propertyValues {Curve {keys [<at 0 1 RCIM_Cubic>,<at 0.5 0.75 RCIM_Cubic>,<at 1 1 RCIM_Cubic>]}}}
     CurveIndex=$Particles.PartitionParticles.Normalized Index In Partition
     Scale Curve=&Multiply_Float(
      A=40
      B=&FloatFromCurve(
       FloatCurve=DataInterface() {propertyValues {Curve {keys [
           {interpMode=RCIM_Cubic, leaveTangentWeight=0.1728, tangentMode=RCTM_User, time=0.1, value=0.65}
           {
            arriveTangent=2.6077
            arriveTangentWeight=1.3568
            interpMode=RCIM_Cubic
            leaveTangent=-31.7719
            leaveTangentWeight=0.4846
            tangentMode=RCTM_Break
            time=0.45
            value=1
           }
          ]}}}
       CurveIndex=&Divide_Float(A=&VectorLength(Vector=$Engine.Owner.Velocity), B=300)
      )
     )
    )
    Target Position Value:NiagaraPosition=$Particles.Position
    Variable Type=::NewEnumerator6
   )
   /SetVariables(Particles.RibbonShape:NiagaraPosition=&AddVectorToPosition(
     Position=&AddVectorToPosition(
      Position=$Particles.InterpolateOverTime.InterpolateOverTime_Position.Moving Average
      Vector=&Multiply_Vector(A=$Particles.VectortoCamera, B=&VectorFromFloat(Value=$Particles.CameraOffsetRibbon))
     )
     Vector=&AddVectorToPosition(
      Position=&MergeHelper_VectorSwitch(
       0=&MakeVector(
        X=&FloatFromCurve(
         FloatCurve=DataInterface() {propertyValues {Curve {
            defaultValue=0
            keys [
             {arriveTangent=6.8692, interpMode=RCIM_Cubic, leaveTangent=6.8692, tangentMode=RCTM_User, time=0, value=0}
             {arriveTangent=0.1758, interpMode=RCIM_Cubic, leaveTangent=0.1758, time=0.1717, value=0.9182}
             {interpMode=RCIM_Cubic, tangentMode=RCTM_User, time=0.4953, value=0.0871}
             {arriveTangent=-0.1725, interpMode=RCIM_Cubic, leaveTangent=-0.1725, time=0.8022, value=0.9371}
             {arriveTangent=-7.5341, tangentMode=RCTM_User, time=1, value=0}
            ]
           }}}
         CurveIndex=$Particles.PartitionParticles.Normalized Index In Partition
         Scale Curve=&Multiply_Float(
          A=$Particles.CurveAdjust
          B=&Lerp_Float(
           A=7
           B=-7
           Alpha=&Waveform(
            Global Phase=&Frac_Float(Float=&Add_Float(A=$Particles.PartitionParticles.Normalized Index In Partition, B=&Multiply_Float(A=$Engine.Time, B=0.07)))
            "Waveform [1]"=::NewEnumerator2
            "[1] Frequency"=1
            "[1] Amplitude Min/Max"=<0 1>
           )
          )
         )
        )
        Y=&FloatFromCurve(
         FloatCurve=DataInterface() {propertyValues {Curve {
            defaultValue=0
            keys [
             {arriveTangent=6.8692, interpMode=RCIM_Cubic, leaveTangent=6.8692, tangentMode=RCTM_User, time=0, value=0}
             {arriveTangent=0.1758, interpMode=RCIM_Cubic, leaveTangent=0.1758, time=0.1717, value=0.9182}
             {interpMode=RCIM_Cubic, tangentMode=RCTM_User, time=0.4953, value=0.0871}
             {arriveTangent=-0.1725, interpMode=RCIM_Cubic, leaveTangent=-0.1725, time=0.8022, value=0.9371}
             {arriveTangent=-7.5341, tangentMode=RCTM_User, time=1, value=0}
            ]
           }}}
         CurveIndex=$Particles.PartitionParticles.Normalized Index In Partition
         Scale Curve=&Multiply_Float(
          A=$Particles.CurveAdjust
          B=&Lerp_Float(
           A=-10
           B=10
           Alpha=&Waveform(
            Global Phase=&Frac_Float(Float=&Add_Float(A=$Particles.PartitionParticles.Normalized Index In Partition, B=&Multiply_Float(A=$Engine.Time, B=0.05)))
            "Waveform [1]"=::NewEnumerator2
            "[1] Phase Scale"=1.5
            "[1] Frequency"=1
            "[1] Amplitude Min/Max"=<0 1>
           )
          )
         )
        )
        Z=&Multiply_Float(
         A=&FloatFromCurve(
          FloatCurve=DataInterface() {propertyValues {Curve {
             defaultValue=0
             keys [
              {arriveTangent=4.0936, interpMode=RCIM_Cubic, leaveTangent=4.0936, tangentMode=RCTM_User, time=0, value=0}
              {interpMode=RCIM_Cubic, tangentMode=RCTM_User, time=0.5, value=1}
              {arriveTangent=-3.0128, tangentMode=RCTM_User, time=1, value=0}
             ]
            }}}
          CurveIndex=$Particles.PartitionParticles.Normalized Index In Partition
          Scale Curve=&Lerp_Float(A=32, B=40, Alpha=$Particles.Waveform1)
         )
         B=$Particles.HeightAdjust
        )
       )
       1=&MakeVector(
        X=&FloatFromCurve(
         FloatCurve=DataInterface() {propertyValues {Curve {
            defaultValue=0
            keys [
             {arriveTangent=6.8692, interpMode=RCIM_Cubic, leaveTangent=6.8692, tangentMode=RCTM_User, time=0, value=0}
             {arriveTangent=0.1741, interpMode=RCIM_Cubic, leaveTangent=0.1741, time=0.1717, value=0.9182}
             {interpMode=RCIM_Cubic, tangentMode=RCTM_User, time=0.5, value=0.0871}
             {arriveTangent=-0.1741, interpMode=RCIM_Cubic, leaveTangent=-0.1741, time=0.8022, value=0.9371}
             {arriveTangent=-7.5341, tangentMode=RCTM_User, time=1, value=0}
            ]
           }}}
         CurveIndex=$Particles.PartitionParticles.Normalized Index In Partition
         Scale Curve=&Multiply_Float(
          A=$Particles.CurveAdjust
          B=&Lerp_Float(
           A=10
           B=-10
           Alpha=&Waveform(
            Global Phase=&Frac_Float(Float=&Add_Float(A=$Particles.PartitionParticles.Normalized Index In Partition, B=&Multiply_Float(A=$Engine.Time, B=0.045)))
            "Waveform [1]"=::NewEnumerator2
            "[1] Phase Scale"=1.5
            "[1] Frequency"=1
            "[1] Amplitude Min/Max"=<0 1>
           )
          )
         )
        )
        Y=&FloatFromCurve(
         FloatCurve=DataInterface() {propertyValues {Curve {
            defaultValue=0
            keys [
             {arriveTangent=6.8692, interpMode=RCIM_Cubic, leaveTangent=6.8692, tangentMode=RCTM_User, time=0, value=0}
             {arriveTangent=0.1741, interpMode=RCIM_Cubic, leaveTangent=0.1741, time=0.1717, value=0.9182}
             {interpMode=RCIM_Cubic, tangentMode=RCTM_User, time=0.5, value=0.0871}
             {arriveTangent=-0.1741, interpMode=RCIM_Cubic, leaveTangent=-0.1741, time=0.8022, value=0.9371}
             {arriveTangent=-7.5341, tangentMode=RCTM_User, time=1, value=0}
            ]
           }}}
         CurveIndex=$Particles.PartitionParticles.Normalized Index In Partition
         Scale Curve=&Multiply_Float(
          A=$Particles.CurveAdjust
          B=&Lerp_Float(
           A=-8
           B=8
           Alpha=&Waveform(
            Global Phase=&Frac_Float(Float=&Add_Float(A=$Particles.PartitionParticles.Normalized Index In Partition, B=&Multiply_Float(A=$Engine.Time, B=0.06)))
            "Waveform [1]"=::NewEnumerator2
            "[1] Frequency"=1
            "[1] Amplitude Min/Max"=<0 1>
           )
          )
         )
        )
        Z=&Multiply_Float(
         A=&FloatFromCurve(
          FloatCurve=DataInterface() {propertyValues {Curve {
             defaultValue=0
             keys [
              {arriveTangent=4.0936, interpMode=RCIM_Cubic, leaveTangent=4.0936, tangentMode=RCTM_User, time=0, value=0}
              {interpMode=RCIM_Cubic, tangentMode=RCTM_User, time=0.5, value=1}
              {arriveTangent=-3.0128, tangentMode=RCTM_User, time=1, value=0}
             ]
            }}}
          CurveIndex=$Particles.PartitionParticles.Normalized Index In Partition
          Scale Curve=&Lerp_Float(A=42, B=30, Alpha=$Particles.Waveform1)
         )
         B=$Particles.HeightAdjust
        )
       )
       2=?
       3=?
       Index=$Particles.PartitionParticles.Partition
       Num=::NewEnumerator0
      )
      Vector=&Add_Vector(
       A=&Multiply_Vector(
        A=&MergeHelper_VectorSwitch(
         0=&VectorFromFloat(Value=&Multiply_Float(
           A=&FloatFromCurve(
            FloatCurve=DataInterface() {propertyValues {Curve {
               defaultValue=1
               keys [
                {arriveTangent=3.4053, interpMode=RCIM_Cubic, leaveTangent=3.4053, tangentMode=RCTM_User, time=0, value=0}
                {arriveTangentWeight=0.08, interpMode=RCIM_Cubic, leaveTangentWeight=0.3468, tangentMode=RCTM_User, time=0.15, value=1}
                {arriveTangentWeight=0.2167, interpMode=RCIM_Cubic, leaveTangentWeight=0.0667, tangentMode=RCTM_User, time=0.85, value=1}
               ]
              }}}
            CurveIndex=$Particles.PartitionParticles.Normalized Index In Partition
            Scale Curve=$Particles.WidthAdjust
           )
           B=&Multiply_Float(
            A=&FloatFromCurve(
             FloatCurve=DataInterface() {propertyValues {Curve {
                defaultValue=0
                keys [
                 {interpMode=RCIM_Cubic, leaveTangentWeight=0.2361, tangentMode=RCTM_User, time=0.05, value=0}
                 {
                  arriveTangent=0.7804
                  arriveTangentWeight=0.2787
                  interpMode=RCIM_Cubic
                  leaveTangent=0.7804
                  leaveTangentWeight=0.0872
                  tangentMode=RCTM_User
                  time=0.7455
                  value=1
                 }
                 {arriveTangent=-11.0422, arriveTangentWeight=0.7206, interpMode=RCIM_Cubic, leaveTangent=-11.0422, tangentMode=RCTM_User, time=1, value=0.1}
                ]
               }}}
             CurveIndex=$Particles.PartitionParticles.Normalized Index In Partition
             Scale Curve=-29
            )
            B=1
           )
          ))
         1=&VectorFromFloat(Value=&Multiply_Float(
           A=&FloatFromCurve(
            FloatCurve=DataInterface() {propertyValues {Curve {
               defaultValue=1
               keys [
                {arriveTangent=3.4053, interpMode=RCIM_Cubic, leaveTangent=3.4053, tangentMode=RCTM_User, time=0, value=0}
                {arriveTangentWeight=0.08, interpMode=RCIM_Cubic, leaveTangentWeight=0.3468, tangentMode=RCTM_User, time=0.15, value=1}
                {arriveTangentWeight=0.2167, interpMode=RCIM_Cubic, leaveTangentWeight=0.0667, tangentMode=RCTM_User, time=0.85, value=1}
               ]
              }}}
            CurveIndex=$Particles.PartitionParticles.Normalized Index In Partition
            Scale Curve=$Particles.WidthAdjust
           )
           B=&Multiply_Float(
            A=&FloatFromCurve(
             FloatCurve=DataInterface() {propertyValues {Curve {
                defaultValue=0
                keys [
                 {interpMode=RCIM_Cubic, leaveTangentWeight=0.2633, tangentMode=RCTM_User, time=0.1, value=0}
                 {arriveTangentWeight=0.2787, interpMode=RCIM_Cubic, leaveTangentWeight=0.0872, tangentMode=RCTM_User, time=0.724, value=1}
                 {arriveTangent=-12.3048, arriveTangentWeight=0.7206, interpMode=RCIM_Cubic, leaveTangent=-12.3048, tangentMode=RCTM_User, time=1, value=0}
                ]
               }}}
             CurveIndex=$Particles.PartitionParticles.Normalized Index In Partition
             Scale Curve=25
            )
            B=1
           )
          ))
         2=?
         3=?
         Index=$Particles.PartitionParticles.Partition
         Num=::NewEnumerator0
        )
        B=$System.CameraRight_Vector
       )
       B=&Multiply_Vector(
        A=$System.CameraUp_Vector
        B=&VectorFromFloat(Value=&FloatFromCurve(
          FloatCurve=DataInterface() {propertyValues {Curve {
             defaultValue=1
             keys [
              {interpMode=RCIM_Cubic, tangentMode=RCTM_User, time=0, value=0.02}
              {interpMode=RCIM_Cubic, tangentMode=RCTM_User, time=0.5, value=1}
              {tangentMode=RCTM_User, time=1, value=0.02}
             ]
            }}}
          CurveIndex=$Particles.PartitionParticles.Normalized Index In Partition
          Scale Curve=&Lerp_Float(
           A=0
           B=18
           Alpha=&MakeCustomFloatFromBool(
            Bool=&SetBoolByFloatComparison(
             A=&OneMinusFloat(Float=&DotProduct(Vector A=$Particles.VectortoCamera, Vector B=<0 0 1>))
             B=0.5
            )
            True Float=&FloatFromCurve(
             FloatCurve=DataInterface() {propertyValues {Curve {keys [<at 0.5 0>,<at 1 1>]}}}
             CurveIndex=&Abs_Float(Float:NiagaraFloat=&DotProduct(Vector A=$Particles.VectortoCamera, Vector B=<0 0 1>))
             Scale Curve=2
            )
            False Float=0
           )
          )
         ))
       )
      )
     )
    )) {isSetParameters=true, module=SetVariables_#8}
   /SetVariables(Particles.RibbonShape:NiagaraPosition=&AddVectorToPosition(
     Position=&AddVectorToPosition(
      Position=$Particles.InterpolateOverTime.InterpolateOverTime_Position.Moving Average
      Vector=&Multiply_Vector(A=$Particles.VectortoCamera, B=&VectorFromFloat(Value=$Particles.CameraOffsetRibbon))
     )
     Vector=&AddVectorToPosition(
      Position=&MergeHelper_VectorSwitch(
       0=&MakeVector(
        X=&FloatFromCurve(
         FloatCurve=DataInterface() {propertyValues {Curve {
            defaultValue=0
            keys [
             {arriveTangent=6.8692, interpMode=RCIM_Cubic, leaveTangent=6.8692, tangentMode=RCTM_User, time=0, value=0}
             {arriveTangent=0.1758, interpMode=RCIM_Cubic, leaveTangent=0.1758, time=0.1717, value=0.9182}
             {interpMode=RCIM_Cubic, tangentMode=RCTM_User, time=0.4953, value=0.0871}
             {arriveTangent=-0.1725, interpMode=RCIM_Cubic, leaveTangent=-0.1725, time=0.8022, value=0.9371}
             {arriveTangent=-7.5341, tangentMode=RCTM_User, time=1, value=0}
            ]
           }}}
         CurveIndex=$Particles.PartitionParticles.Normalized Index In Partition
         Scale Curve=&Multiply_Float(
          A=$Particles.CurveAdjust
          B=&Lerp_Float(
           A=7
           B=-7
           Alpha=&Waveform(
            Global Phase=&Frac_Float(Float=&Add_Float(A=$Particles.PartitionParticles.Normalized Index In Partition, B=&Multiply_Float(A=$Engine.Time, B=0.07)))
            "Waveform [1]"=::NewEnumerator2
            "[1] Frequency"=1
            "[1] Amplitude Min/Max"=<0 1>
           )
          )
         )
        )
        Y=&FloatFromCurve(
         FloatCurve=DataInterface() {propertyValues {Curve {
            defaultValue=0
            keys [
             {arriveTangent=6.8692, interpMode=RCIM_Cubic, leaveTangent=6.8692, tangentMode=RCTM_User, time=0, value=0}
             {arriveTangent=0.1758, interpMode=RCIM_Cubic, leaveTangent=0.1758, time=0.1717, value=0.9182}
             {interpMode=RCIM_Cubic, tangentMode=RCTM_User, time=0.4953, value=0.0871}
             {arriveTangent=-0.1725, interpMode=RCIM_Cubic, leaveTangent=-0.1725, time=0.8022, value=0.9371}
             {arriveTangent=-7.5341, tangentMode=RCTM_User, time=1, value=0}
            ]
           }}}
         CurveIndex=$Particles.PartitionParticles.Normalized Index In Partition
         Scale Curve=&Multiply_Float(
          A=$Particles.CurveAdjust
          B=&Lerp_Float(
           A=-10
           B=10
           Alpha=&Waveform(
            Global Phase=&Frac_Float(Float=&Add_Float(A=$Particles.PartitionParticles.Normalized Index In Partition, B=&Multiply_Float(A=$Engine.Time, B=0.05)))
            "Waveform [1]"=::NewEnumerator2
            "[1] Phase Scale"=1.5
            "[1] Frequency"=1
            "[1] Amplitude Min/Max"=<0 1>
           )
          )
         )
        )
        Z=&Multiply_Float(
         A=&FloatFromCurve(
          FloatCurve=DataInterface() {propertyValues {Curve {
             defaultValue=0
             keys [
              {arriveTangent=4.0936, interpMode=RCIM_Cubic, leaveTangent=4.0936, tangentMode=RCTM_User, time=0, value=0}
              {interpMode=RCIM_Cubic, tangentMode=RCTM_User, time=0.5, value=1}
              {arriveTangent=-3.0128, tangentMode=RCTM_User, time=1, value=0}
             ]
            }}}
          CurveIndex=$Particles.PartitionParticles.Normalized Index In Partition
          Scale Curve=&Lerp_Float(A=0, B=0, Alpha=$Particles.Waveform1)
         )
         B=$Particles.HeightAdjust
        )
       )
       1=&MakeVector(
        X=&FloatFromCurve(
         FloatCurve=DataInterface() {propertyValues {Curve {
            defaultValue=0
            keys [
             {arriveTangent=6.8692, interpMode=RCIM_Cubic, leaveTangent=6.8692, tangentMode=RCTM_User, time=0, value=0}
             {arriveTangent=0.1741, interpMode=RCIM_Cubic, leaveTangent=0.1741, time=0.1717, value=0.9182}
             {interpMode=RCIM_Cubic, tangentMode=RCTM_User, time=0.5, value=0.0871}
             {arriveTangent=-0.1741, interpMode=RCIM_Cubic, leaveTangent=-0.1741, time=0.8022, value=0.9371}
             {arriveTangent=-7.5341, tangentMode=RCTM_User, time=1, value=0}
            ]
           }}}
         CurveIndex=$Particles.PartitionParticles.Normalized Index In Partition
         Scale Curve=&Multiply_Float(
          A=$Particles.CurveAdjust
          B=&Lerp_Float(
           A=10
           B=-10
           Alpha=&Waveform(
            Global Phase=&Frac_Float(Float=&Add_Float(A=$Particles.PartitionParticles.Normalized Index In Partition, B=&Multiply_Float(A=$Engine.Time, B=0.045)))
            "Waveform [1]"=::NewEnumerator2
            "[1] Phase Scale"=1.5
            "[1] Frequency"=1
            "[1] Amplitude Min/Max"=<0 1>
           )
          )
         )
        )
        Y=&FloatFromCurve(
         FloatCurve=DataInterface() {propertyValues {Curve {
            defaultValue=0
            keys [
             {arriveTangent=6.8692, interpMode=RCIM_Cubic, leaveTangent=6.8692, tangentMode=RCTM_User, time=0, value=0}
             {arriveTangent=0.1741, interpMode=RCIM_Cubic, leaveTangent=0.1741, time=0.1717, value=0.9182}
             {interpMode=RCIM_Cubic, tangentMode=RCTM_User, time=0.5, value=0.0871}
             {arriveTangent=-0.1741, interpMode=RCIM_Cubic, leaveTangent=-0.1741, time=0.8022, value=0.9371}
             {arriveTangent=-7.5341, tangentMode=RCTM_User, time=1, value=0}
            ]
           }}}
         CurveIndex=$Particles.PartitionParticles.Normalized Index In Partition
         Scale Curve=&Multiply_Float(
          A=$Particles.CurveAdjust
          B=&Lerp_Float(
           A=-8
           B=8
           Alpha=&Waveform(
            Global Phase=&Frac_Float(Float=&Add_Float(A=$Particles.PartitionParticles.Normalized Index In Partition, B=&Multiply_Float(A=$Engine.Time, B=0.06)))
            "Waveform [1]"=::NewEnumerator2
            "[1] Frequency"=1
            "[1] Amplitude Min/Max"=<0 1>
           )
          )
         )
        )
        Z=&Multiply_Float(
         A=&FloatFromCurve(
          FloatCurve=DataInterface() {propertyValues {Curve {
             defaultValue=0
             keys [
              {arriveTangent=4.0936, interpMode=RCIM_Cubic, leaveTangent=4.0936, tangentMode=RCTM_User, time=0, value=0}
              {interpMode=RCIM_Cubic, tangentMode=RCTM_User, time=0.5, value=1}
              {arriveTangent=-3.0128, tangentMode=RCTM_User, time=1, value=0}
             ]
            }}}
          CurveIndex=$Particles.PartitionParticles.Normalized Index In Partition
          Scale Curve=&Lerp_Float(A=0, B=0, Alpha=$Particles.Waveform1)
         )
         B=$Particles.HeightAdjust
        )
       )
       2=?
       3=?
       Index=$Particles.PartitionParticles.Partition
       Num=::NewEnumerator0
      )
      Vector=&Add_Vector(
       A=&Multiply_Vector(
        A=&MergeHelper_VectorSwitch(
         0=&VectorFromFloat(Value=&Multiply_Float(
           A=&FloatFromCurve(
            FloatCurve=DataInterface() {propertyValues {Curve {
               defaultValue=1
               keys [
                {arriveTangent=3.4053, interpMode=RCIM_Cubic, leaveTangent=3.4053, tangentMode=RCTM_User, time=0, value=0}
                {arriveTangentWeight=0.08, interpMode=RCIM_Cubic, leaveTangentWeight=0.3468, tangentMode=RCTM_User, time=0.15, value=1}
                {arriveTangentWeight=0.2167, interpMode=RCIM_Cubic, leaveTangentWeight=0.0667, tangentMode=RCTM_User, time=0.85, value=1}
               ]
              }}}
            CurveIndex=$Particles.PartitionParticles.Normalized Index In Partition
            Scale Curve=$Particles.WidthAdjust
           )
           B=&Multiply_Float(
            A=&FloatFromCurve(
             FloatCurve=DataInterface() {propertyValues {Curve {
                defaultValue=0
                keys [
                 {interpMode=RCIM_Cubic, leaveTangentWeight=0.2361, tangentMode=RCTM_User, time=0.05, value=0}
                 {
                  arriveTangent=0.7804
                  arriveTangentWeight=0.2787
                  interpMode=RCIM_Cubic
                  leaveTangent=0.7804
                  leaveTangentWeight=0.0872
                  tangentMode=RCTM_User
                  time=0.7455
                  value=1
                 }
                 {arriveTangent=-11.0422, arriveTangentWeight=0.7206, interpMode=RCIM_Cubic, leaveTangent=-11.0422, tangentMode=RCTM_User, time=1, value=0.1}
                ]
               }}}
             CurveIndex=$Particles.PartitionParticles.Normalized Index In Partition
             Scale Curve=-29
            )
            B=1
           )
          ))
         1=&VectorFromFloat(Value=&Multiply_Float(
           A=&FloatFromCurve(
            FloatCurve=DataInterface() {propertyValues {Curve {
               defaultValue=1
               keys [
                {arriveTangent=3.4053, interpMode=RCIM_Cubic, leaveTangent=3.4053, tangentMode=RCTM_User, time=0, value=0}
                {arriveTangentWeight=0.08, interpMode=RCIM_Cubic, leaveTangentWeight=0.3468, tangentMode=RCTM_User, time=0.15, value=1}
                {arriveTangentWeight=0.2167, interpMode=RCIM_Cubic, leaveTangentWeight=0.0667, tangentMode=RCTM_User, time=0.85, value=1}
               ]
              }}}
            CurveIndex=$Particles.PartitionParticles.Normalized Index In Partition
            Scale Curve=$Particles.WidthAdjust
           )
           B=&Multiply_Float(
            A=&FloatFromCurve(
             FloatCurve=DataInterface() {propertyValues {Curve {
                defaultValue=0
                keys [
                 {interpMode=RCIM_Cubic, leaveTangentWeight=0.2633, tangentMode=RCTM_User, time=0.1, value=0}
                 {arriveTangentWeight=0.2787, interpMode=RCIM_Cubic, leaveTangentWeight=0.0872, tangentMode=RCTM_User, time=0.724, value=1}
                 {arriveTangent=-12.3048, arriveTangentWeight=0.7206, interpMode=RCIM_Cubic, leaveTangent=-12.3048, tangentMode=RCTM_User, time=1, value=0}
                ]
               }}}
             CurveIndex=$Particles.PartitionParticles.Normalized Index In Partition
             Scale Curve=25
            )
            B=1
           )
          ))
         2=?
         3=?
         Index=$Particles.PartitionParticles.Partition
         Num=::NewEnumerator0
        )
        B=$System.CameraRight_Vector
       )
       B=&Multiply_Vector(
        A=$System.CameraUp_Vector
        B=&VectorFromFloat(Value=&FloatFromCurve(
          FloatCurve=DataInterface() {propertyValues {Curve {
             defaultValue=1
             keys [
              {interpMode=RCIM_Cubic, tangentMode=RCTM_User, time=0, value=0.02}
              {interpMode=RCIM_Cubic, tangentMode=RCTM_User, time=0.5, value=1}
              {tangentMode=RCTM_User, time=1, value=0.02}
             ]
            }}}
          CurveIndex=$Particles.PartitionParticles.Normalized Index In Partition
          Scale Curve=&Lerp_Float(
           A=0
           B=18
           Alpha=&MakeCustomFloatFromBool(
            Bool=&SetBoolByFloatComparison(
             A=&OneMinusFloat(Float=&DotProduct(Vector A=$Particles.VectortoCamera, Vector B=<0 0 1>))
             B=0.5
            )
            True Float=&FloatFromCurve(
             FloatCurve=DataInterface() {propertyValues {Curve {keys [<at 0.5 0>,<at 1 1>]}}}
             CurveIndex=&Abs_Float(Float:NiagaraFloat=&DotProduct(Vector A=$Particles.VectortoCamera, Vector B=<0 0 1>))
             Scale Curve=2
            )
            False Float=0
           )
          )
         ))
       )
      )
     )
    )) {isSetParameters=true, module=SetVariables_#9}
  ]
  rendererClasses [&NiagaraRibbonRendererProperties]
  renderers [{rendererClass=&NiagaraRibbonRendererProperties}]
 }
 Decal_GlowRays {
  emitterUpdateScript [
   &EmitterState()
   &SpawnBurst_Instantaneous(Spawn Count=5, Use Spawn Probability=false, UseLoopCountLimit=true)
   &SpawnBurst_Instantaneous(Spawn Count=3, Spawn Group=1, Use Spawn Probability=false, UseLoopCountLimit=true) {module=SpawnBurst_Instantaneous001}
   &SpawnBurst_Instantaneous(Spawn Count=3, Spawn Group=2, Use Spawn Probability=false, UseLoopCountLimit=true) {module=SpawnBurst_Instantaneous002}
   &SpawnBurst_Instantaneous(Spawn Count=2, Spawn Group=3, Use Spawn Probability=false, UseLoopCountLimit=true) {module=SpawnBurst_Instantaneous003}
  ]
  particleSpawnScript [
   /SetVariables(Particles.VisibilityTag=&SpawnGroup()) {isSetParameters=true}
   &PartitionParticles(Number Of Partitions=&MergeHelper_IntSwitch(0=5, 1=3, 2=3, 3=2, 4=?, 5=?, Index=&SpawnGroup(), Num=::NewEnumerator2))
   &InitializeParticle(
    Color=&MergeHelper_ColorSwitch(
     0=&Multiply_LinearColorByFloat(Linear Color=<rgba 1.75 1.75 1.75 1>, Float=2)
     1=&MultiplyLinearColor(
      Linear Color=&MergeHelper_ColorSwitch(0=<rgba 8 8 8 1>, 1=<rgba 4 4 4 1>, 2=<rgba 4 4 4 1>, 3=?, Index=$Particles.PartitionParticles.Partition, Num=::NewEnumerator1)
      Scale Factor=&MakeVector4_FromFloat(Float=&MakeCustomFloatFromBool(Bool=$User.bIsFrontEndPreview, True Float=1, False Float=1))
     )
     2=&MultiplyLinearColor(
      Linear Color=&MergeHelper_ColorSwitch(0=<rgba 3 3 3 1>, 1=<rgba 2 2 2 1>, 2=<rgba 2 2 2 1>, 3=?, Index=$Particles.PartitionParticles.Partition, Num=::NewEnumerator1)
      Scale Factor=&MakeVector4_FromFloat(Float=&MakeCustomFloatFromBool(Bool=$User.bIsFrontEndPreview, True Float=1, False Float=1))
     )
     3=&MakeCustomLinearColorFromBool(Bool=$User.bIsFrontEndPreview, True Color=<rgba 0.1953 0.1953 0.1953 1>, False Color=<rgba 0.1198 0.1198 0.1198 1>)
     Index=&SpawnGroup()
     Num=::NewEnumerator2
    )
    Sprite Rotation Mode=::NewEnumerator0
    Sprite Size=&MergeHelper_Vector2Switch(
     0=&Vector2DFromFloat(Value=&Multiply_Float(
       A=&SelectFloatFromArray(
        Direct Array Index=$Particles.PartitionParticles.Partition
        Float Selection Array=DataInterface() {propertyValues {FloatData [50,45,45,35,35]}}
       )
       B=&MakeCustomFloatFromBool(Bool=$User.IsLargeBody, True Float=1.5, False Float=1)
      ))
     1=&Vector2DFromFloat(Value=&Multiply_Float(
       A=&SelectFloatFromArray(Direct Array Index=$Particles.PartitionParticles.Partition, Float Selection Array=DataInterface() {propertyValues {FloatData [135,95,95]}})
       B=&MakeCustomFloatFromBool(Bool=$User.IsLargeBody, True Float=1.2, False Float=1)
      ))
     2=&Vector2DFromFloat(Value=&Multiply_Float(
       A=&SelectFloatFromArray(Direct Array Index=$Particles.PartitionParticles.Partition, Float Selection Array=DataInterface() {propertyValues {FloatData [95,80,80]}})
       B=&MakeCustomFloatFromBool(Bool=$User.IsLargeBody, True Float=1.2, False Float=1)
      ))
     3=&Multiply_Vector2D_ByFloat(
      Float=&MakeCustomFloatFromBool(Bool=$User.IsLargeBody, True Float=1.2, False Float=1)
      Vector2D=&Multiply_Vector2D_ByFloat(Float=48, Vector2D=<1 1.3>)
     )
     4=?
     Index=&SpawnGroup()
     Num=::NewEnumerator2
    )
    Sprite Size Mode=::NewEnumerator4
    UsePositionOffset=false
    Write Material Random=true
   ) {module=InitializeParticle001}
   &CameraOffset(Camera Offset Amount=&MergeHelper_FloatSwitch(
     0=&SelectFloatFromArray(Direct Array Index=$Particles.PartitionParticles.Partition, Float Selection Array=DataInterface() {propertyValues {FloatData [5,7,7,5,5]}})
     1=-65
     2=-57
     3=-20
     Index=&SpawnGroup()
     Num=::NewEnumerator2
    ))
  ]
  particleUpdateScript [
   &ParticleState(Kill Particles When Lifetime Has Elapsed=false, Let Infinitely Lived Particles Die When Emitter Deactivates=true)
   &SkeletalMeshLocation(
    Bone / Socket Index=&MergeHelper_IntSwitch(
     0=&SelectIntFromArray(Direct Array Index=$Particles.PartitionParticles.Partition, Int Selection Array=DataInterface() {propertyValues {IntData [0,1,2,3,4]}})
     1=&SelectIntFromArray(Direct Array Index=$Particles.PartitionParticles.Partition, Int Selection Array=DataInterface() {propertyValues {IntData [7,1,2]}})
     2=&SelectIntFromArray(Direct Array Index=$Particles.PartitionParticles.Partition, Int Selection Array=DataInterface() {propertyValues {IntData [7,1,2]}})
     3=&SelectIntFromArray(Direct Array Index=$Particles.PartitionParticles.Partition, Int Selection Array=DataInterface() {propertyValues {IntData [1,2]}})
     4=?
     5=?
     Index=&SpawnGroup()
     Num=::NewEnumerator2
    )
    Bone Sampling Mode=::NewEnumerator4
    MaskSpawnGroup=false
    Orientation Sampling=::NewEnumerator1
    Sampled Position Offset=&MergeHelper_VectorSwitch(
     0=<0 0 0>
     1=&Multiply_Vector(
      A=&Multiply_VectorByFloat(
       Vector=&SelectVectorFromArray(
        Direct Array Index=$Particles.PartitionParticles.Partition
        Vector Selection Array=DataInterface() {propertyValues {FloatData [<7 0 0>, <3 0 0>, <-3 0 0>]}}
       )
       Float=1
      )
      B=&MakeCustomVectorFromBool(Bool=$User.IsLargeBody, True Vector=<1 1 1.25>, False Vector=<1 1 1>)
     )
     2=&Multiply_Vector(
      A=&Multiply_VectorByFloat(
       Vector=&SelectVectorFromArray(
        Direct Array Index=$Particles.PartitionParticles.Partition
        Vector Selection Array=DataInterface() {propertyValues {FloatData [<15 0 0>, <-2 0 0>, <2 0 0>]}}
       )
       Float=1
      )
      B=&MakeCustomVectorFromBool(Bool=$User.IsLargeBody, True Vector=<1 1 1.25>, False Vector=<1 1 1>)
     )
     3=&MergeHelper_VectorSwitch(0=<2 0 0>, 1=<-2 0 0>, 2=?, 3=?, Index=$Particles.PartitionParticles.Partition, Num=::NewEnumerator0)
     Index=&SpawnGroup()
     Num=::NewEnumerator2
    )
    Skeletal Mesh=$System.Skeletal Mesh
    Use Sampled Mesh Attribute as Ribbon ID=false
   ) {module=SkeletalMeshLocation003}
   &DynamicMaterialParameters(
    Index 0 Param 1=&SelectFloatFromArray(
     Direct Array Index=$Particles.PartitionParticles.Partition
     Float Selection Array=DataInterface() {propertyValues {FloatData [1.5,1.75,1.75,2,2]}}
    )
    Module.Index0Float0Write=true
    Module.Index0Float1Write=true
    Module.Index0Float2Write=true
    Module.Index0Float3Write=true
   )
   &SpriteRotationRate(Rotation Rate=&MergeHelper_FloatSwitch(
     0=0
     1=&MakeCustomFloatFromBool(
      Bool=&SetBoolByIntComparison(A=$Particles.PartitionParticles.Partition, B=0, Comparison Type=::NewEnumerator0)
      True Float=&RandomRangeFloat(Minimum=10, Maximum=15)
      False Float=&RandomRangeFloat(Minimum=-10, Maximum=-15)
     )
     2=?
     3=?
     Index=&SpawnGroup()
     Num=::NewEnumerator0
    ))
   &ScaleColor(
    Scale Alpha=&MergeHelper_FloatSwitch(
     0=6
     1=&FloatFromCurve(
      FloatCurve=DataInterface() {propertyValues {Curve {keys [
          <at -0.0009 0.85 RCIM_Cubic>
          {arriveTangent=0.4254, interpMode=RCIM_Cubic, leaveTangent=0.4254, time=0.1742, value=0.95}
          {arriveTangentWeight=0.0596, interpMode=RCIM_Cubic, leaveTangentWeight=0.0823, tangentMode=RCTM_User, time=0.3518, value=1}
          {arriveTangentWeight=0.0838, interpMode=RCIM_Cubic, leaveTangentWeight=0.1397, tangentMode=RCTM_User, time=0.5968, value=1}
          <at 1.0052 0.85 RCIM_Cubic>
         ]}}}
      CurveIndex=Linked() {linkedVariable {name=NPC.FortniteNPC.TimeOfDay_0-1, type=NiagaraFloat}}
     )
     2=1
     3=&Multiply_Float(A=&Waveform("Waveform [1]"=::NewEnumerator0, "[1] Frequency"=1, "[1] Amplitude Min/Max"=<0.9 1>), B=1)
     Index=&SpawnGroup()
     Num=::NewEnumerator2
    )
    Scale RGB=&Multiply_VectorByFloat(
     Vector=&MakeVectorFromLinearColorRGB(Color=$System.ColorShift)
     Float=&MergeHelper_FloatSwitch(
      0=0.35
      1=2
      2=&FloatFromCurve(
       FloatCurve=DataInterface() {propertyValues {Curve {keys [
           <at -0.0009 0.9 RCIM_Cubic>
           {arriveTangent=0.2836, interpMode=RCIM_Cubic, leaveTangent=0.2836, time=0.1742, value=0.95}
           {arriveTangentWeight=0.0596, interpMode=RCIM_Cubic, leaveTangentWeight=0.0823, tangentMode=RCTM_User, time=0.3518, value=1}
           {arriveTangentWeight=0.0838, interpMode=RCIM_Cubic, leaveTangentWeight=0.1397, tangentMode=RCTM_User, time=0.5968, value=1}
           <at 1.0052 0.9 RCIM_Cubic>
          ]}}}
       CurveIndex=Linked() {linkedVariable {name=NPC.FortniteNPC.TimeOfDay_0-1, type=NiagaraFloat}}
      )
      3=1
      Index=&SpawnGroup()
      Num=::NewEnumerator2
     )
    )
    ScaleA=true
    ScaleRGB=true
   )
   /SetVariables(Particles.Flare Dyn0:Vector4f=&MakeVector4(X=0, Y=0, Z=0, W=0)) {isSetParameters=true}
   /SetVariables(Particles.VectortoCamera:Vector3f=&ReturnCameraVector(Camera Vectors=::NewEnumerator4, Coordinate Space=::Local, Origin Position=$Particles.Position)) {isSetParameters=true}
   &InterpolateOverTime(
    Rate Of Change=0.5
    Target Float Value:NiagaraFloat=&MakeCustomFloatFromBool(
     Bool=&SetBoolByFloatComparison(A=&DotProduct(Vector A=$Particles.VectortoCamera, Vector B=<0 1 0>), B=0)
     True Float=&FloatFromCurve(
      FloatCurve=DataInterface() {propertyValues {Curve {keys [<at 0 0>,<at 1 1>]}}}
      CurveIndex=&Abs_Float(Float:NiagaraFloat=&DotProduct(Vector A=$Particles.VectortoCamera, Vector B=<0 1 0>))
     )
     False Float=&FloatFromCurve(
      FloatCurve=DataInterface() {propertyValues {Curve {defaultValue=1, keys [{tangentMode=RCTM_User, time=0, value=1}, {tangentMode=RCTM_User, time=1, value=0}]}}}
      CurveIndex=&Abs_Float(Float:NiagaraFloat=&DotProduct(Vector A=$Particles.VectortoCamera, Vector B=<0 1 0>))
     )
    )
   )
   /SetVariables(
    Particles.PositionMod:NiagaraPosition=&AddVectorToPosition(
     Position=$Particles.Position
     Vector=&Multiply_VectorByFloat(
      Vector=$System.CameraUp_Vector
      Float=&MergeHelper_FloatSwitch(
       0=0
       1=0
       2=2
       3=&MakeCustomFloatFromBool(Bool=$User.IsLargeBody, True Float=3, False Float=2)
       Index=&SpawnGroup()
       Num=::NewEnumerator2
      )
     )
    )
    Particles.SpriteRotationMod:NiagaraFloat=&MergeHelper_FloatSwitch(
     0=0
     1=0
     2=&Multiply_Float(
      A=&MergeHelper_FloatSwitch(
       0=0
       1=&Lerp_Float(A=1, B=-1, Alpha=$Particles.InterpolateOverTime.InterpolateOverTime_Float.Moving Average)
       2=&Lerp_Float(A=-1, B=1, Alpha=$Particles.InterpolateOverTime.InterpolateOverTime_Float.Moving Average)
       3=?
       Index=$Particles.PartitionParticles.Partition
       Num=::NewEnumerator1
      )
      B=-55
     )
     3=&Multiply_Float(
      A=&MergeHelper_FloatSwitch(
       0=&Lerp_Float(A=-1, B=1, Alpha=$Particles.InterpolateOverTime.InterpolateOverTime_Float.Moving Average)
       1=&Lerp_Float(A=1, B=-1, Alpha=$Particles.InterpolateOverTime.InterpolateOverTime_Float.Moving Average)
       2=?
       3=?
       Index=$Particles.PartitionParticles.Partition
       Num=::NewEnumerator0
      )
      B=70
     )
     Index=&SpawnGroup()
     Num=::NewEnumerator2
    )
   ) {isSetParameters=true}
  ]
  rendererClasses [&NiagaraSpriteRendererProperties]
  renderers [
   {rendererClass=&NiagaraSpriteRendererProperties}
   {rendererClass=&NiagaraSpriteRendererProperties, rendererIndex=1}
   {rendererClass=&NiagaraSpriteRendererProperties, rendererIndex=2}
   {rendererClass=&NiagaraSpriteRendererProperties, rendererIndex=3}
  ]
 }
 LiquidMesh {
  emitterUpdateScript [
   &EmitterState()
   &SpawnBurst_Instantaneous(Spawn Count=2, Use Spawn Probability=false, UseLoopCountLimit=true)
  ]
  particleSpawnScript [
   &PartitionParticles(Particles Per Partition=1, Partition Mode:ENiagara_PartitionMode_v2=::NewEnumerator1)
   &InitializeParticle(
    Lifetime=5
    Mesh Scale Mode=::NewEnumerator3
    Mesh Uniform Scale=&Multiply_Float(
     A=&MergeHelper_FloatSwitch(0=1.6, 1=1.7, 2=?, 3=?, Index=$Particles.PartitionParticles.Partition, Num=::NewEnumerator0)
     B=&MakeCustomFloatFromBool(Bool=$User.IsLargeBody, True Float=1.16, False Float=1.1)
    )
    UsePositionOffset=false
    Write Material Random=true
   )
   &CameraOffset(Camera Offset Amount=&MakeCustomFloatFromBool(Bool=$User.IsLargeBody, True Float=-50, False Float=-50))
   &InitialMeshOrientation(
    Mesh Orientation Mode=::NewEnumerator2
    Rotation=&MergeHelper_VectorSwitch(0=<0 0 0>, 1=<0.4864 0 0>, 2=?, 3=?, Index=$Particles.PartitionParticles.Partition, Num=::NewEnumerator0)
    Use Rotation Vector=true
   )
   /SetVariables(Particles.VisibilityTag=$Particles.PartitionParticles.Partition) {isSetParameters=true}
  ]
  particleUpdateScript [
   &ParticleState(Kill Particles When Lifetime Has Elapsed=false, Let Infinitely Lived Particles Die When Emitter Deactivates=true)
   /SetVariables(Particles.OffsetVertical=12) {isSetParameters=true}
   &SkeletalMeshLocation(
    Bone / Socket Index=7
    Bone Sampling Mode=::NewEnumerator4
    MaskSpawnGroup=false
    Orientation Sampling=::NewEnumerator1
    Sampled Position Offset=&Multiply_VectorByFloat(
     Vector=&AddVectorToPosition(
      Position=&MakeVector(X=$Particles.OffsetVertical, Y=0, Z=0)
      Vector=&MakeCustomVectorFromBool(Bool=$User.bIsFrontEndPreview, True Vector=<1 0 0>, False Vector=<0 0 0>)
     )
     Float=&MakeCustomFloatFromBool(Bool=$User.IsLargeBody, True Float=1.15, False Float=1)
    )
    Skeletal Mesh=$System.Skeletal Mesh
    Use Sampled Mesh Attribute as Ribbon ID=false
   )
   &UpdateMeshOrientation(
    Facing Direction=&RotateVectorByQuaternion(Vector To Rotate=<0 1 0>, Quaternion=$Particles.SkeletalMeshLocation.SampledRotation)
    Facing Interpolation=false
    Orientation Method=::NewEnumerator1
    Override Quat=false
    TransformFacing=false
    TransformUp=false
   ) {enabled=false, module=UpdateMeshOrientation001}
   &UpdateMeshOrientation(
    Rotation Rate=&Waveform(
     Waveform Count=::NewEnumerator1
     "Waveform [1]"=::NewEnumerator12
     "[1] Phase Scale"=0.5
     "[1] Frequency"=1
     "[1] Amplitude Min/Max"=<-0.1 0.1>
     "[1] Scale"=&MergeHelper_FloatSwitch(0=0.75, 1=0.6, 2=?, 3=?, Index=$Particles.PartitionParticles.Partition, Num=::NewEnumerator0)
     "[2] Average"=false
     "[2] Input Position"=0
     "[2] Phase"=$Emitter.Age
     "[2] Phase Scale"=0.1
     "[2] Compound Frequency"=2
     "[2] Frequency"=2
     "[2] Amplitude Min/Max"=<-0.1 0.1>
     "[2] Scale"=&MergeHelper_FloatSwitch(0=0.3, 1=0.55, 2=?, 3=?, Index=$Particles.PartitionParticles.Partition, Num=::NewEnumerator0)
     "[2] Bias"=0
     "[2] Exponent"=1
     "[2] Pulse Duration"=0.1
     "[2] Clip Triangle Min/Max"=<0 0>
     "[2] Random Seed"=0
    )
    Rotation Vector=&MakeVector(
     X=1
     Y=&MergeHelper_FloatSwitch(0=0, 1=0.4, 2=?, 3=?, Index=$Particles.PartitionParticles.Partition, Num=::NewEnumerator0)
     Z=&MergeHelper_FloatSwitch(0=0.4, 1=0, 2=?, 3=?, Index=$Particles.PartitionParticles.Partition, Num=::NewEnumerator0)
    )
   ) {enabled=false}
   &UpdateMeshOrientation(
    Rotation Rate=&Waveform(
     Global Input Position=&MergeHelper_FloatSwitch(0=0, 1=0.07, 2=?, 3=?, Index=$Particles.PartitionParticles.Partition, Num=::NewEnumerator0)
     "Waveform [1]"=::NewEnumerator0
     "[1] Phase Scale"=0.1
     "[1] Frequency"=1
     "[1] Amplitude Min/Max"=<-0.15 0.15>
     "[1] Scale"=0.7
     "[2] Average"=false
     "[2] Input Position"=0
     "[2] Phase"=$Emitter.Age
     "[2] Frequency"=2
     "[2] Amplitude Min/Max"=<-0.1 0.1>
     "[2] Scale"=&MergeHelper_FloatSwitch(0=0.3, 1=0.5, 2=?, 3=?, Index=$Particles.PartitionParticles.Partition, Num=::NewEnumerator0)
     "[2] Bias"=0
     "[2] Exponent"=1
    )
    Rotation Vector=&MakeVector(
     X=1
     Y=&MergeHelper_FloatSwitch(0=0, 1=0.4, 2=?, 3=?, Index=$Particles.PartitionParticles.Partition, Num=::NewEnumerator0)
     Z=&MergeHelper_FloatSwitch(0=0.4, 1=0, 2=?, 3=?, Index=$Particles.PartitionParticles.Partition, Num=::NewEnumerator0)
    )
   ) {module=UpdateMeshOrientation002}
   &ScaleMeshSize(Scale Factor=&VectorFromFloat(Value=&Waveform(
      Global Phase=&ReturnNormalizedExecIndex()
      "Waveform [1]"=::NewEnumerator0
      "[1] Phase Scale"=0.5
      "[1] Frequency"=1
      "[1] Amplitude Min/Max"=<0.95 1.1>
     )))
   &DynamicMaterialParameters(
    Index 0 Param 1=&Multiply_Float(
     A=&Waveform(
      "Waveform [1]"=::NewEnumerator0
      "[1] Phase Scale"=0.1
      "[1] Frequency"=1
      "[1] Amplitude Min/Max"=&MergeHelper_Vector2Switch(0=<1.25 1.5>, 1=<1.5 1.25>, 2=?, 3=?, 4=?, Index=$Particles.PartitionParticles.Partition, Num=::NewEnumerator0)
     )
     B=1.5
    )
    Module.Index0Float0Write=true
    Module.Index0Float1Write=true
    Module.Index0Float2Write=true
    Module.Index0Float3Write=true
   )
   &InterpolateOverTime(
    Rate Of Change:NiagaraFloat=&Multiply_Float(
     A=&MergeHelper_FloatSwitch(0=40, 1=35, 2=?, 3=?, Index=$Particles.PartitionParticles.Partition, Num=::NewEnumerator0)
     B=&FloatFromCurve(
      FloatCurve=DataInterface() {propertyValues {Curve {keys [
          {interpMode=RCIM_Cubic, leaveTangentWeight=0.1728, tangentMode=RCTM_User, time=0.1, value=0.65}
          {
           arriveTangent=2.6077
           arriveTangentWeight=1.3568
           interpMode=RCIM_Cubic
           leaveTangent=-31.7719
           leaveTangentWeight=0.4846
           tangentMode=RCTM_Break
           time=0.45
           value=1
          }
         ]}}}
      CurveIndex=&Divide_Float(A=&VectorLength(Vector=$Engine.Owner.Velocity), B=300)
     )
    )
    Target Position Value:NiagaraPosition=$Particles.Position
    Variable Type=::NewEnumerator6
   )
   &ScaleColor(
    Scale RGB=&MakeVectorFromLinearColorRGB(Color=&Multiply_LinearColorByFloat(Linear Color=$System.ColorShift, Float=0.5))
    ScaleA=true
    ScaleRGB=true
   )
  ]
  rendererClasses [&NiagaraMeshRendererProperties]
  renderers [
   {rendererClass=&NiagaraMeshRendererProperties, rendererIndex=0}
   {rendererClass=&NiagaraMeshRendererProperties, rendererIndex=1}
  ]
 }
 Orbit_Blobs {
  emitterUpdateScript [
   &EmitterState()
   &SpawnRate(Spawn Probability=0.7, SpawnRate=4, Use Spawn Probability=true)
   &SpawnBurst_Instantaneous(Spawn Count=3, Use Spawn Probability=false, UseLoopCountLimit=true)
  ]
  particleSpawnScript [
   &InitializeParticle(
    Lifetime=&RandomRangeFloat(Minimum=4, Maximum=4.5)
    Mesh Scale=&Multiply_VectorByFloat(
     Vector=&RandomRangeVector(Minimum=<0.9 0.9 1>, Maximum=<1.2 1.2 1.2>)
     Float=&FloatFromCurve(FloatCurve=DataInterface() {propertyValues {Curve {keys [<at 0 1.5>,<at 0.7602 1.7332>,<at 1 2.75>]}}})
    )
    Mesh Scale Mode=::NewEnumerator4
    UsePositionOffset=false
    Write Material Random=true
   )
   &ShapeLocation(
    Apply To Particle Position=::NewEnumerator1
    MaskSpawnGroup=false
    Override Seed=false
    Shape Origin=&VectorFromFloat(Value=0)
    Sphere Radius=5
    Sphere Surface Distribution=0.3214
    UseOwnerScale=true
   )
   &CameraOffset(Camera Offset Amount=-50)
   /SetVariables(Particles.MeshIndex=/Niagara/DynamicInputs/Arrays/SelectIntFromWeightedDistributionArray(
     Array Sampling Mode:ENiagara_LifetimeMode=::NewEnumerator1
     Direct Array Index:NiagaraInt32=?
     Direct Array Mode:ENiagara_WrapClamp=::NewEnumerator0
     Evaluation Type:ENiagaraRandomnessEvaluation=::NewEnumerator0
     Recalculate Random Each Loop:NiagaraBool=false
     Weighted Distribution Int Array:NiagaraDataInterfaceArrayDistributionInt=DataInterface() {propertyValues {ArrayData [{value=0, weight=0.25}, {value=1, weight=0.75}]}}
     Randomness Mode:ENiagaraRandomnessMode=::NewEnumerator0
     Random Seed:NiagaraInt32=0
     Fixed Random Seed:NiagaraBool=false
    )) {isSetParameters=true}
  ]
  particleUpdateScript [
   &ParticleState()
   &SkeletalMeshLocation(
    Bone / Socket Index=7
    Bone Sampling Mode=::NewEnumerator4
    MaskSpawnGroup=false
    Orientation Sampling=::NewEnumerator1
    Position Sampling=::NewEnumerator1
    Sampled Position Offset=&AddVectorToPosition(Position=<9 0 0>, Vector=Linked() {linkedVariable {name=Particles.Initial.ShapeLocation.ShapePosition, type=NiagaraPosition}})
    Skeletal Mesh=$System.Skeletal Mesh
    Use Sampled Mesh Attribute as Ribbon ID=false
   )
   &ScaleColor(
    Scale RGB=&MakeVectorFromLinearColorRGB(Color=&Multiply_LinearColorByFloat(Linear Color=$System.ColorShift, Float=0.5))
    ScaleA=true
    ScaleRGB=true
   )
   &ScaleMeshSize(Scale Factor=&VectorFromCurve(VectorCurve=DataInterface() {propertyValues {
       XCurve {keys [
         {arriveTangent=-0.5011, interpMode=RCIM_Cubic, time=0.001, value=1.2818}
         {arriveTangentWeight=0.0634, interpMode=RCIM_Cubic, leaveTangentWeight=0.0856, tangentMode=RCTM_User, time=0.2944, value=0.7538}
         {arriveTangentWeight=0.0575, interpMode=RCIM_Cubic, leaveTangentWeight=0.0883, tangentMode=RCTM_User, time=0.5669, value=1.1681}
         {arriveTangentWeight=0.0761, interpMode=RCIM_Cubic, leaveTangentWeight=0.0767, tangentMode=RCTM_User, time=0.825, value=0.8979}
         <at 1.0016 1.0434 RCIM_Cubic>
        ]}
       YCurve {keys [
         {arriveTangent=-0.5011, interpMode=RCIM_Cubic, time=0.001, value=1.2818}
         {arriveTangentWeight=0.0634, interpMode=RCIM_Cubic, leaveTangentWeight=0.0856, tangentMode=RCTM_User, time=0.2944, value=0.7538}
         {arriveTangentWeight=0.0575, interpMode=RCIM_Cubic, leaveTangentWeight=0.0883, tangentMode=RCTM_User, time=0.5669, value=1.1681}
         {arriveTangentWeight=0.0761, interpMode=RCIM_Cubic, leaveTangentWeight=0.0767, tangentMode=RCTM_User, time=0.825, value=0.8979}
         <at 1.0016 1.0434 RCIM_Cubic>
        ]}
       ZCurve {keys [
         {arriveTangentWeight=0.1414, interpMode=RCIM_Cubic, leaveTangentWeight=0.2506, tangentMode=RCTM_User, time=0.0148, value=0.6908}
         {arriveTangentWeight=0.0704, interpMode=RCIM_Cubic, leaveTangentWeight=0.1047, tangentMode=RCTM_User, time=0.2944, value=1.2947}
         {arriveTangentWeight=0.068, interpMode=RCIM_Cubic, leaveTangentWeight=0.0957, tangentMode=RCTM_User, time=0.5669, value=0.8695}
         {arriveTangentWeight=0.0648, interpMode=RCIM_Cubic, leaveTangentWeight=0.0692, tangentMode=RCTM_User, time=0.825, value=1.1477}
         <at 1 0.9752 RCIM_Cubic>
        ]}
      }})) {module=ScaleMeshSize001}
   &ScaleMeshSize(Scale Factor=&Multiply_VectorByFloat(
     Vector=&VectorFromFloat(Value=&FloatFromCurve(FloatCurve=DataInterface() {propertyValues {Curve {
          defaultValue=1
          keys [
           {arriveTangent=5.0323, interpMode=RCIM_Cubic, leaveTangent=5.0323, tangentMode=RCTM_User, time=0, value=0}
           {
            arriveTangent=0.6284
            arriveTangentWeight=0.0667
            interpMode=RCIM_Cubic
            leaveTangent=0.6284
            leaveTangentWeight=0.1667
            tangentMode=RCTM_User
            time=0.25
            value=1
           }
           {arriveTangent=-1.3002, interpMode=RCIM_Cubic, leaveTangent=-1.3002, tangentMode=RCTM_User, time=0.4621, value=0.718}
           {
            arriveTangent=1.1179
            arriveTangentWeight=0.2668
            interpMode=RCIM_Cubic
            leaveTangent=1.1179
            leaveTangentWeight=0.1601
            tangentMode=RCTM_User
            time=0.8885
            value=0.6075
           }
           {arriveTangent=-5.449, interpMode=RCIM_Cubic, leaveTangent=-5.449, time=0.9385, value=0.6708}
           <at 1 0 RCIM_Cubic>
          ]
         }}}))
     Float=&MakeFloatFromVector(
      Vector=&VectorFromFloat(Value=&Waveform(
        Global Input Position=&RandomRangeFloat(Minimum=0, Maximum=1)
        "Waveform [1]"=::NewEnumerator0
        "[1] Frequency"=1
        "[1] Amplitude Min/Max"=<0.95 1.05>
       ))
      Channel=::NewEnumerator0
     )
    ))
   &DynamicMaterialParameters(
    Index 0 Param 1=0.1
    Module.Index0Float0Write=true
    Module.Index0Float1Write=true
    Module.Index0Float2Write=true
    Module.Index0Float3Write=true
   )
   &RotateAroundPoint(
    Radius:NiagaraFloat=&Multiply_Float(
     A=&RandomRangeFloat(Minimum=18, Maximum=30)
     B=&FloatFromCurve(FloatCurve=DataInterface() {propertyValues {Curve {
         defaultValue=1
         keys [
          {arriveTangent=3.1874, interpMode=RCIM_Cubic, leaveTangent=3.1874, tangentMode=RCTM_User, time=0, value=0.5}
          {arriveTangentWeight=0.1333, interpMode=RCIM_Cubic, leaveTangentWeight=0.2, tangentMode=RCTM_User, time=0.4, value=1}
          {arriveTangent=-0.8627, arriveTangentWeight=0.3958, interpMode=RCIM_Cubic, leaveTangent=-0.8627, tangentMode=RCTM_User, time=1, value=0.75}
         ]
        }}})
    )
    Rotation Center:NiagaraPosition=$Particles.SkeletalMeshLocation.SampledPosition
    Rotation Phase:NiagaraFloat=&RandomRangeFloat(Minimum=0, Maximum=1)
    Rotation Rate:NiagaraFloat=&Multiply_Float(
     A=&Multiply_Float(
      A=&RandomRangeFloat(Minimum=0.07, Maximum=0.1)
      B=&FloatFromCurve(FloatCurve=DataInterface() {propertyValues {Curve {
          defaultValue=1
          keys [
           {tangentMode=RCTM_User, time=0, value=1}
           {arriveTangent=-0.0405, interpMode=RCIM_Cubic, leaveTangent=-0.0405, tangentMode=RCTM_User, time=0.6, value=1}
           {arriveTangent=-5.0069, interpMode=RCIM_Cubic, leaveTangent=-5.0069, tangentMode=RCTM_User, time=1, value=0}
          ]
         }}})
     )
     B=&SelectFloatFromArray(
      Array Sampling Mode=::NewEnumerator0
      Direct Array Index=?
      Float Selection Array=DataInterface() {propertyValues {FloatData [-1,1]}}
      Random Seed=0
     )
    )
    X Axis:Vector3f=$System.CameraUp_Vector
    Y Axis:Vector3f=$System.CameraRight_Vector
   )
   &UpdateMeshOrientation(
    Facing Direction=$Particles.RotateAroundPoint.VectorFromCenter
    Facing Interpolation=true
    Facing Interpolation Rate=2
    Orientation Method=::NewEnumerator1
    Override Quat=false
    TransformFacing=false
    TransformUp=false
   ) {module=UpdateMeshOrientation002}
  ]
  rendererClasses [&NiagaraMeshRendererProperties]
  renderers [{rendererClass=&NiagaraMeshRendererProperties, rendererIndex=0}]
 }
 RibbonWorldZOffset {
  bEnabled=false
  emitterSpawnScript [/SetVariables(
    Emitter.NumberOfRibbons=2
    Emitter.ParticlesPerRibbon:NiagaraInt32=&MakeCustomIntFromBool(Bool=$User.IsLargeBody, True INT=20, False INT=20)
   ) {isSetParameters=true, module=SetVariables_#4}]
  emitterUpdateScript [
   &EmitterState()
   &SpawnBurst_Instantaneous(
    Spawn Count=&Multiply_Int(A=$Emitter.NumberOfRibbons, B=$Emitter.ParticlesPerRibbon)
    Use Spawn Probability=false
    UseLoopCountLimit=true
   )
   &SpawnRate(Use Spawn Probability=false) {enabled=false}
  ]
  particleSpawnScript [
   &PartitionParticles(
    Normalized Index In Partition Includes One=true
    Number Of Partitions=$Emitter.NumberOfRibbons
    Particles Per Partition:NiagaraInt32=$Emitter.ParticlesPerRibbon
    Partition Mode:ENiagara_PartitionMode_v2=::NewEnumerator1
    Write Partition to Ribbon ID=::NewEnumerator0
   )
   &InitializeParticle(
    Lifetime=5
    Ribbon Width=8
    Ribbon Width Mode=::NewEnumerator1
    Sprite Size Mode=::NewEnumerator3
    UsePositionOffset=false
    Write Material Random=true
   )
  ]
  particleUpdateScript [
   &ParticleState(Kill Particles When Lifetime Has Elapsed=false, Let Infinitely Lived Particles Die When Emitter Deactivates=true)
   /SetVariables(
    Particles.CurveAdjust=1.2
    Particles.HeightAdjust:NiagaraFloat=&Multiply_Float(
     A=&MakeCustomFloatFromBool(Bool=$User.IsLargeBody, True Float=1.25, False Float=1.1)
     B=&MakeCustomFloatFromBool(Bool=$User.bIsFrontEndPreview, True Float=1.1, False Float=1)
    )
    Particles.Waveform1:NiagaraFloat=&Waveform("Waveform [1]"=::NewEnumerator0, "[1] Phase Scale"=0.1, "[1] Frequency"=1, "[1] Amplitude Min/Max"=<0 1>)
    Particles.Waveform2:NiagaraFloat=&Waveform(
     "Waveform [1]"=::NewEnumerator0
     "[1] Phase Scale"=2.5
     "[1] Phase"=&Add_Float(A=$Particles.PartitionParticles.Normalized Index In Partition, B=&Multiply_Float(A=$Engine.Time, B=0.15))
     "[1] Frequency"=1
     "[1] Amplitude Min/Max"=<0 1>
     "Waveform [2]"=::NewEnumerator12
    )
    Particles.WidthAdjust:NiagaraFloat=&MakeCustomFloatFromBool(Bool=$User.IsLargeBody, True Float=1.25, False Float=1.1)
   ) {isSetParameters=true, module=SetVariables_#5}
   &SkeletalMeshLocation(
    Bone / Socket Index=7
    Bone Sampling Mode=::NewEnumerator4
    MaskSpawnGroup=false
    Orientation Sampling=::NewEnumerator1
    Sampled Position Offset=&MergeHelper_VectorSwitch(
     0=&AddVectorToPosition(
      Position=<-15 0 0>
      Vector=&MakeVector(
       X=&FloatFromCurve(
        FloatCurve=DataInterface() {propertyValues {Curve {
           defaultValue=0
           keys [{interpMode=RCIM_Cubic, tangentMode=RCTM_User, time=0, value=1}, {interpMode=RCIM_Cubic, tangentMode=RCTM_User, time=1, value=0}]
          }}}
        CurveIndex=$Particles.PartitionParticles.Normalized Index In Partition
        Scale Curve=5
       )
       Y=0
       Z=0
      )
     )
     1=&AddVectorToPosition(
      Position=<-15 0 0>
      Vector=&MakeVector(
       X=&FloatFromCurve(
        FloatCurve=DataInterface() {propertyValues {Curve {
           defaultValue=0
           keys [{interpMode=RCIM_Cubic, tangentMode=RCTM_User, time=0, value=1}, {interpMode=RCIM_Cubic, tangentMode=RCTM_User, time=1, value=0}]
          }}}
        CurveIndex=$Particles.PartitionParticles.Normalized Index In Partition
        Scale Curve=5
       )
       Y=0
       Z=0
      )
     )
     2=?
     3=?
     Index=$Particles.PartitionParticles.Partition
     Num=::NewEnumerator0
    )
    Skeletal Mesh=$System.Skeletal Mesh
    Use Sampled Mesh Attribute as Ribbon ID=false
   )
   &ScaleColor(
    Scale RGB=&Multiply_Vector(
     A=&MakeVectorFromLinearColorRGB(Color=&ColorFromCurve(
       ColorCurve=DataInterface() {propertyValues {
         AlphaCurve {keys [<at 0 1 RCIM_Cubic>]}
         BlueCurve {keys [<at 0.0229 0.0212 RCIM_Cubic>,<at 0.1897 0.2747>,<at 0.4637 0.2747>,<at 0.5356 0.2747>,<at 0.7777 0.2747>,<at 0.9715 0.0212 RCIM_Cubic>]}
         GreenCurve {keys [<at 0.0229 0.0529 RCIM_Cubic>,<at 0.1897 0.0438>,<at 0.4637 0>,<at 0.5356 0>,<at 0.7777 0.0595>,<at 0.9715 0.0529 RCIM_Cubic>]}
         RedCurve {keys [<at 0.0229 0 RCIM_Cubic>,<at 0.1897 0>,<at 0.4637 0.117>,<at 0.5356 0.117>,<at 0.7777 0>,<at 0.9715 0 RCIM_Cubic>]}
        }}
       CurveIndex=&Frac_Float(Float=&Add_Float(A=$Particles.PartitionParticles.Normalized Index In Partition, B=&Multiply_Float(A=$Engine.Time, B=0.5)))
      ))
     B=&VectorFromFloat(Value=0.3)
    )
    ScaleA=true
    ScaleRGB=true
   )
   &ScaleRibbonWidth(Ribbon Width Scale=&Multiply_Float(
     A=&MergeHelper_FloatSwitch(
      0=&Lerp_Float(A=0.1, B=1.25, Alpha=$Particles.Waveform2)
      1=&Lerp_Float(A=1.25, B=0.1, Alpha=$Particles.Waveform2)
      2=?
      3=?
      Index=$Particles.PartitionParticles.Partition
      Num=::NewEnumerator0
     )
     B=&FloatFromCurve(
      FloatCurve=DataInterface() {propertyValues {Curve {keys [
          {arriveTangent=4.1029, interpMode=RCIM_Cubic, leaveTangent=4.1029, tangentMode=RCTM_User, time=0, value=0.5}
          {arriveTangentWeight=0.2426, interpMode=RCIM_Cubic, leaveTangentWeight=0.082, tangentMode=RCTM_User, time=0.15, value=1}
          {arriveTangentWeight=0.1132, interpMode=RCIM_Cubic, leaveTangentWeight=0.2253, tangentMode=RCTM_User, time=0.2007, value=1}
          {arriveTangent=0.1507, interpMode=RCIM_Cubic, leaveTangent=0.675, tangentMode=RCTM_Break, time=0.5007, value=0.5}
          {arriveTangentWeight=0.2258, interpMode=RCIM_Cubic, leaveTangentWeight=0.1124, tangentMode=RCTM_User, time=0.8007, value=1}
          {arriveTangentWeight=0.135, interpMode=RCIM_Cubic, leaveTangentWeight=0.0442, tangentMode=RCTM_User, time=0.85, value=1}
          {arriveTangent=-8.6495, interpMode=RCIM_Cubic, leaveTangent=-8.6495, tangentMode=RCTM_User, time=0.9989, value=0.5}
         ]}}}
      CurveIndex=$Particles.PartitionParticles.Normalized Index In Partition
      Scale Curve=$System.Timeline.Float
     )
    ))
   /SetVariables(Particles.VectortoCamera:Vector3f=&ReturnCameraVector(Camera Vectors=::NewEnumerator4, Coordinate Space=::Simulation, Origin Position=$Particles.Position)) {isSetParameters=true, module=SetVariables_#6}
   /SetVariables(Particles.CameraOffsetRibbon:NiagaraFloat=&Lerp_Float(
     A=&MakeCustomFloatFromBool(Bool=$User.bIsFrontEndPreview, True Float=-30, False Float=-35)
     B=-15
     Alpha=&MakeCustomFloatFromBool(
      Bool=&SetBoolByFloatComparison(
       A=&OneMinusFloat(Float=&DotProduct(Vector A=$Particles.VectortoCamera, Vector B=<0 0 1>))
       B=0.5
      )
      True Float=&FloatFromCurve(
       FloatCurve=DataInterface() {propertyValues {Curve {keys [<at 0.55 0>,<at 0.95 1>]}}}
       CurveIndex=&Abs_Float(Float:NiagaraFloat=&DotProduct(Vector A=$Particles.VectortoCamera, Vector B=<0 0 1>))
      )
      False Float=0
     )
    )) {isSetParameters=true, module=SetVariables_#7}
   &InterpolateOverTime(
    Rate Of Change:NiagaraFloat=&FloatFromCurve(
     FloatCurve=DataInterface() {propertyValues {Curve {keys [<at 0 1 RCIM_Cubic>,<at 0.5 0.75 RCIM_Cubic>,<at 1 1 RCIM_Cubic>]}}}
     CurveIndex=$Particles.PartitionParticles.Normalized Index In Partition
     Scale Curve=&Multiply_Float(
      A=40
      B=&FloatFromCurve(
       FloatCurve=DataInterface() {propertyValues {Curve {keys [
           {interpMode=RCIM_Cubic, leaveTangentWeight=0.1728, tangentMode=RCTM_User, time=0.1, value=0.65}
           {
            arriveTangent=2.6077
            arriveTangentWeight=1.3568
            interpMode=RCIM_Cubic
            leaveTangent=-31.7719
            leaveTangentWeight=0.4846
            tangentMode=RCTM_Break
            time=0.45
            value=1
           }
          ]}}}
       CurveIndex=&Divide_Float(A=&VectorLength(Vector=$Engine.Owner.Velocity), B=300)
      )
     )
    )
    Target Position Value:NiagaraPosition=$Particles.Position
    Variable Type=::NewEnumerator6
   )
   /SetVariables(Particles.RibbonShape:NiagaraPosition=&AddVectorToPosition(
     Position=&AddVectorToPosition(
      Position=$Particles.InterpolateOverTime.InterpolateOverTime_Position.Moving Average
      Vector=&Multiply_Vector(A=$Particles.VectortoCamera, B=&VectorFromFloat(Value=$Particles.CameraOffsetRibbon))
     )
     Vector=&AddVectorToPosition(
      Position=&MergeHelper_VectorSwitch(
       0=&MakeVector(
        X=&FloatFromCurve(
         FloatCurve=DataInterface() {propertyValues {Curve {
            defaultValue=0
            keys [
             {arriveTangent=6.8692, interpMode=RCIM_Cubic, leaveTangent=6.8692, tangentMode=RCTM_User, time=0, value=0}
             {arriveTangent=0.1758, interpMode=RCIM_Cubic, leaveTangent=0.1758, time=0.1717, value=0.9182}
             {interpMode=RCIM_Cubic, tangentMode=RCTM_User, time=0.4953, value=0.0871}
             {arriveTangent=-0.1725, interpMode=RCIM_Cubic, leaveTangent=-0.1725, time=0.8022, value=0.9371}
             {arriveTangent=-7.5341, tangentMode=RCTM_User, time=1, value=0}
            ]
           }}}
         CurveIndex=$Particles.PartitionParticles.Normalized Index In Partition
         Scale Curve=&Multiply_Float(
          A=$Particles.CurveAdjust
          B=&Lerp_Float(
           A=7
           B=-7
           Alpha=&Waveform(
            Global Phase=&Frac_Float(Float=&Add_Float(A=$Particles.PartitionParticles.Normalized Index In Partition, B=&Multiply_Float(A=$Engine.Time, B=0.07)))
            "Waveform [1]"=::NewEnumerator2
            "[1] Frequency"=1
            "[1] Amplitude Min/Max"=<0 1>
           )
          )
         )
        )
        Y=&FloatFromCurve(
         FloatCurve=DataInterface() {propertyValues {Curve {
            defaultValue=0
            keys [
             {arriveTangent=6.8692, interpMode=RCIM_Cubic, leaveTangent=6.8692, tangentMode=RCTM_User, time=0, value=0}
             {arriveTangent=0.1758, interpMode=RCIM_Cubic, leaveTangent=0.1758, time=0.1717, value=0.9182}
             {interpMode=RCIM_Cubic, tangentMode=RCTM_User, time=0.4953, value=0.0871}
             {arriveTangent=-0.1725, interpMode=RCIM_Cubic, leaveTangent=-0.1725, time=0.8022, value=0.9371}
             {arriveTangent=-7.5341, tangentMode=RCTM_User, time=1, value=0}
            ]
           }}}
         CurveIndex=$Particles.PartitionParticles.Normalized Index In Partition
         Scale Curve=&Multiply_Float(
          A=$Particles.CurveAdjust
          B=&Lerp_Float(
           A=-10
           B=10
           Alpha=&Waveform(
            Global Phase=&Frac_Float(Float=&Add_Float(A=$Particles.PartitionParticles.Normalized Index In Partition, B=&Multiply_Float(A=$Engine.Time, B=0.05)))
            "Waveform [1]"=::NewEnumerator2
            "[1] Phase Scale"=1.5
            "[1] Frequency"=1
            "[1] Amplitude Min/Max"=<0 1>
           )
          )
         )
        )
        Z=&Multiply_Float(
         A=&FloatFromCurve(
          FloatCurve=DataInterface() {propertyValues {Curve {
             defaultValue=0
             keys [
              {arriveTangent=4.0936, interpMode=RCIM_Cubic, leaveTangent=4.0936, tangentMode=RCTM_User, time=0, value=0}
              {interpMode=RCIM_Cubic, tangentMode=RCTM_User, time=0.5, value=1}
              {arriveTangent=-3.0128, tangentMode=RCTM_User, time=1, value=0}
             ]
            }}}
          CurveIndex=$Particles.PartitionParticles.Normalized Index In Partition
          Scale Curve=&Lerp_Float(A=32, B=40, Alpha=$Particles.Waveform1)
         )
         B=$Particles.HeightAdjust
        )
       )
       1=&MakeVector(
        X=&FloatFromCurve(
         FloatCurve=DataInterface() {propertyValues {Curve {
            defaultValue=0
            keys [
             {arriveTangent=6.8692, interpMode=RCIM_Cubic, leaveTangent=6.8692, tangentMode=RCTM_User, time=0, value=0}
             {arriveTangent=0.1741, interpMode=RCIM_Cubic, leaveTangent=0.1741, time=0.1717, value=0.9182}
             {interpMode=RCIM_Cubic, tangentMode=RCTM_User, time=0.5, value=0.0871}
             {arriveTangent=-0.1741, interpMode=RCIM_Cubic, leaveTangent=-0.1741, time=0.8022, value=0.9371}
             {arriveTangent=-7.5341, tangentMode=RCTM_User, time=1, value=0}
            ]
           }}}
         CurveIndex=$Particles.PartitionParticles.Normalized Index In Partition
         Scale Curve=&Multiply_Float(
          A=$Particles.CurveAdjust
          B=&Lerp_Float(
           A=10
           B=-10
           Alpha=&Waveform(
            Global Phase=&Frac_Float(Float=&Add_Float(A=$Particles.PartitionParticles.Normalized Index In Partition, B=&Multiply_Float(A=$Engine.Time, B=0.045)))
            "Waveform [1]"=::NewEnumerator2
            "[1] Phase Scale"=1.5
            "[1] Frequency"=1
            "[1] Amplitude Min/Max"=<0 1>
           )
          )
         )
        )
        Y=&FloatFromCurve(
         FloatCurve=DataInterface() {propertyValues {Curve {
            defaultValue=0
            keys [
             {arriveTangent=6.8692, interpMode=RCIM_Cubic, leaveTangent=6.8692, tangentMode=RCTM_User, time=0, value=0}
             {arriveTangent=0.1741, interpMode=RCIM_Cubic, leaveTangent=0.1741, time=0.1717, value=0.9182}
             {interpMode=RCIM_Cubic, tangentMode=RCTM_User, time=0.5, value=0.0871}
             {arriveTangent=-0.1741, interpMode=RCIM_Cubic, leaveTangent=-0.1741, time=0.8022, value=0.9371}
             {arriveTangent=-7.5341, tangentMode=RCTM_User, time=1, value=0}
            ]
           }}}
         CurveIndex=$Particles.PartitionParticles.Normalized Index In Partition
         Scale Curve=&Multiply_Float(
          A=$Particles.CurveAdjust
          B=&Lerp_Float(
           A=-8
           B=8
           Alpha=&Waveform(
            Global Phase=&Frac_Float(Float=&Add_Float(A=$Particles.PartitionParticles.Normalized Index In Partition, B=&Multiply_Float(A=$Engine.Time, B=0.06)))
            "Waveform [1]"=::NewEnumerator2
            "[1] Frequency"=1
            "[1] Amplitude Min/Max"=<0 1>
           )
          )
         )
        )
        Z=&Multiply_Float(
         A=&FloatFromCurve(
          FloatCurve=DataInterface() {propertyValues {Curve {
             defaultValue=0
             keys [
              {arriveTangent=4.0936, interpMode=RCIM_Cubic, leaveTangent=4.0936, tangentMode=RCTM_User, time=0, value=0}
              {interpMode=RCIM_Cubic, tangentMode=RCTM_User, time=0.5, value=1}
              {arriveTangent=-3.0128, tangentMode=RCTM_User, time=1, value=0}
             ]
            }}}
          CurveIndex=$Particles.PartitionParticles.Normalized Index In Partition
          Scale Curve=&Lerp_Float(A=42, B=30, Alpha=$Particles.Waveform1)
         )
         B=$Particles.HeightAdjust
        )
       )
       2=?
       3=?
       Index=$Particles.PartitionParticles.Partition
       Num=::NewEnumerator0
      )
      Vector=&Add_Vector(
       A=&Multiply_Vector(
        A=&MergeHelper_VectorSwitch(
         0=&VectorFromFloat(Value=&Multiply_Float(
           A=&FloatFromCurve(
            FloatCurve=DataInterface() {propertyValues {Curve {
               defaultValue=1
               keys [
                {arriveTangent=3.4053, interpMode=RCIM_Cubic, leaveTangent=3.4053, tangentMode=RCTM_User, time=0, value=0}
                {arriveTangentWeight=0.08, interpMode=RCIM_Cubic, leaveTangentWeight=0.3468, tangentMode=RCTM_User, time=0.15, value=1}
                {arriveTangentWeight=0.2167, interpMode=RCIM_Cubic, leaveTangentWeight=0.0667, tangentMode=RCTM_User, time=0.85, value=1}
               ]
              }}}
            CurveIndex=$Particles.PartitionParticles.Normalized Index In Partition
            Scale Curve=$Particles.WidthAdjust
           )
           B=&Multiply_Float(
            A=&FloatFromCurve(
             FloatCurve=DataInterface() {propertyValues {Curve {
                defaultValue=0
                keys [
                 {interpMode=RCIM_Cubic, leaveTangentWeight=0.2361, tangentMode=RCTM_User, time=0.05, value=0}
                 {
                  arriveTangent=0.7804
                  arriveTangentWeight=0.2787
                  interpMode=RCIM_Cubic
                  leaveTangent=0.7804
                  leaveTangentWeight=0.0872
                  tangentMode=RCTM_User
                  time=0.7455
                  value=1
                 }
                 {arriveTangent=-11.0422, arriveTangentWeight=0.7206, interpMode=RCIM_Cubic, leaveTangent=-11.0422, tangentMode=RCTM_User, time=1, value=0.1}
                ]
               }}}
             CurveIndex=$Particles.PartitionParticles.Normalized Index In Partition
             Scale Curve=-29
            )
            B=1
           )
          ))
         1=&VectorFromFloat(Value=&Multiply_Float(
           A=&FloatFromCurve(
            FloatCurve=DataInterface() {propertyValues {Curve {
               defaultValue=1
               keys [
                {arriveTangent=3.4053, interpMode=RCIM_Cubic, leaveTangent=3.4053, tangentMode=RCTM_User, time=0, value=0}
                {arriveTangentWeight=0.08, interpMode=RCIM_Cubic, leaveTangentWeight=0.3468, tangentMode=RCTM_User, time=0.15, value=1}
                {arriveTangentWeight=0.2167, interpMode=RCIM_Cubic, leaveTangentWeight=0.0667, tangentMode=RCTM_User, time=0.85, value=1}
               ]
              }}}
            CurveIndex=$Particles.PartitionParticles.Normalized Index In Partition
            Scale Curve=$Particles.WidthAdjust
           )
           B=&Multiply_Float(
            A=&FloatFromCurve(
             FloatCurve=DataInterface() {propertyValues {Curve {
                defaultValue=0
                keys [
                 {interpMode=RCIM_Cubic, leaveTangentWeight=0.2633, tangentMode=RCTM_User, time=0.1, value=0}
                 {arriveTangentWeight=0.2787, interpMode=RCIM_Cubic, leaveTangentWeight=0.0872, tangentMode=RCTM_User, time=0.724, value=1}
                 {arriveTangent=-12.3048, arriveTangentWeight=0.7206, interpMode=RCIM_Cubic, leaveTangent=-12.3048, tangentMode=RCTM_User, time=1, value=0}
                ]
               }}}
             CurveIndex=$Particles.PartitionParticles.Normalized Index In Partition
             Scale Curve=25
            )
            B=1
           )
          ))
         2=?
         3=?
         Index=$Particles.PartitionParticles.Partition
         Num=::NewEnumerator0
        )
        B=$System.CameraRight_Vector
       )
       B=&Multiply_Vector(
        A=$System.CameraUp_Vector
        B=&VectorFromFloat(Value=&FloatFromCurve(
          FloatCurve=DataInterface() {propertyValues {Curve {
             defaultValue=1
             keys [
              {interpMode=RCIM_Cubic, tangentMode=RCTM_User, time=0, value=0.02}
              {interpMode=RCIM_Cubic, tangentMode=RCTM_User, time=0.5, value=1}
              {tangentMode=RCTM_User, time=1, value=0.02}
             ]
            }}}
          CurveIndex=$Particles.PartitionParticles.Normalized Index In Partition
          Scale Curve=&Lerp_Float(
           A=0
           B=18
           Alpha=&MakeCustomFloatFromBool(
            Bool=&SetBoolByFloatComparison(
             A=&OneMinusFloat(Float=&DotProduct(Vector A=$Particles.VectortoCamera, Vector B=<0 0 1>))
             B=0.5
            )
            True Float=&FloatFromCurve(
             FloatCurve=DataInterface() {propertyValues {Curve {keys [<at 0.5 0>,<at 1 1>]}}}
             CurveIndex=&Abs_Float(Float:NiagaraFloat=&DotProduct(Vector A=$Particles.VectortoCamera, Vector B=<0 0 1>))
             Scale Curve=2
            )
            False Float=0
           )
          )
         ))
       )
      )
     )
    )) {isSetParameters=true, module=SetVariables_#8}
  ]
  rendererClasses [&NiagaraRibbonRendererProperties]
  renderers [{rendererClass=&NiagaraRibbonRendererProperties}]
 }
 Ribbon_HybridOffset {
  bEnabled=false
  emitterSpawnScript [/SetVariables(
    Emitter.NumberOfRibbons=2
    Emitter.ParticlesPerRibbon:NiagaraInt32=&MakeCustomIntFromBool(Bool=$User.IsLargeBody, True INT=20, False INT=20)
   ) {isSetParameters=true, module=SetVariables_#4}]
  emitterUpdateScript [
   &EmitterState()
   &SpawnBurst_Instantaneous(
    Spawn Count=&Multiply_Int(A=$Emitter.NumberOfRibbons, B=$Emitter.ParticlesPerRibbon)
    Use Spawn Probability=false
    UseLoopCountLimit=true
   )
   &SpawnRate(Use Spawn Probability=false) {enabled=false}
  ]
  particleSpawnScript [
   &PartitionParticles(
    Normalized Index In Partition Includes One=true
    Number Of Partitions=$Emitter.NumberOfRibbons
    Particles Per Partition:NiagaraInt32=$Emitter.ParticlesPerRibbon
    Partition Mode:ENiagara_PartitionMode_v2=::NewEnumerator1
    Write Partition to Ribbon ID=::NewEnumerator0
   )
   &InitializeParticle(
    Lifetime=5
    Ribbon Width=8
    Ribbon Width Mode=::NewEnumerator1
    Sprite Size Mode=::NewEnumerator3
    UsePositionOffset=false
    Write Material Random=true
   )
  ]
  particleUpdateScript [
   &ParticleState(Kill Particles When Lifetime Has Elapsed=false, Let Infinitely Lived Particles Die When Emitter Deactivates=true)
   /SetVariables(
    Particles.CurveAdjust=1.2
    Particles.HeightAdjust:NiagaraFloat=&Multiply_Float(
     A=&MakeCustomFloatFromBool(Bool=$User.IsLargeBody, True Float=1.25, False Float=1.1)
     B=&MakeCustomFloatFromBool(Bool=$User.bIsFrontEndPreview, True Float=1.1, False Float=1)
    )
    Particles.Waveform1:NiagaraFloat=&Waveform("Waveform [1]"=::NewEnumerator0, "[1] Phase Scale"=0.1, "[1] Frequency"=1, "[1] Amplitude Min/Max"=<0 1>)
    Particles.Waveform2:NiagaraFloat=&Waveform(
     "Waveform [1]"=::NewEnumerator0
     "[1] Phase Scale"=2.5
     "[1] Phase"=&Add_Float(A=$Particles.PartitionParticles.Normalized Index In Partition, B=&Multiply_Float(A=$Engine.Time, B=0.15))
     "[1] Frequency"=1
     "[1] Amplitude Min/Max"=<0 1>
     "Waveform [2]"=::NewEnumerator12
    )
    Particles.WidthAdjust:NiagaraFloat=&MakeCustomFloatFromBool(Bool=$User.IsLargeBody, True Float=1.25, False Float=1.1)
   ) {isSetParameters=true, module=SetVariables_#5}
   &SkeletalMeshLocation(
    Bone / Socket Index=7
    Bone Sampling Mode=::NewEnumerator4
    MaskSpawnGroup=false
    Orientation Sampling=::NewEnumerator1
    Sampled Position Offset=&MergeHelper_VectorSwitch(
     0=&AddVectorToPosition(
      Position=<-15 0 0>
      Vector=&MakeVector(
       X=&Add_Float(
        A=&FloatFromCurve(
         FloatCurve=DataInterface() {propertyValues {Curve {
            defaultValue=0
            keys [{interpMode=RCIM_Cubic, tangentMode=RCTM_User, time=0, value=1}, {interpMode=RCIM_Cubic, tangentMode=RCTM_User, time=1, value=0}]
           }}}
         CurveIndex=$Particles.PartitionParticles.Normalized Index In Partition
         Scale Curve=5
        )
        B=&Multiply_Float(
         A=&FloatFromCurve(
          FloatCurve=DataInterface() {propertyValues {Curve {
             defaultValue=0
             keys [
              {arriveTangent=4.0936, interpMode=RCIM_Cubic, leaveTangent=4.0936, tangentMode=RCTM_User, time=0, value=0}
              {interpMode=RCIM_Cubic, tangentMode=RCTM_User, time=0.5, value=1}
              {arriveTangent=-3.0128, tangentMode=RCTM_User, time=1, value=0}
             ]
            }}}
          CurveIndex=$Particles.PartitionParticles.Normalized Index In Partition
          Scale Curve=&Lerp_Float(A=16, B=20, Alpha=$Particles.Waveform1)
         )
         B=$Particles.HeightAdjust
        )
       )
       Y=0
       Z=0
      )
     )
     1=&AddVectorToPosition(
      Position=<-15 0 0>
      Vector=&MakeVector(
       X=&Add_Float(
        A=&FloatFromCurve(
         FloatCurve=DataInterface() {propertyValues {Curve {
            defaultValue=0
            keys [{interpMode=RCIM_Cubic, tangentMode=RCTM_User, time=0, value=1}, {interpMode=RCIM_Cubic, tangentMode=RCTM_User, time=1, value=0}]
           }}}
         CurveIndex=$Particles.PartitionParticles.Normalized Index In Partition
         Scale Curve=5
        )
        B=&FloatFromCurve(
         FloatCurve=DataInterface() {propertyValues {Curve {
            defaultValue=0
            keys [
             {arriveTangent=4.0936, interpMode=RCIM_Cubic, leaveTangent=4.0936, tangentMode=RCTM_User, time=0, value=0}
             {interpMode=RCIM_Cubic, tangentMode=RCTM_User, time=0.5, value=1}
             {arriveTangent=-3.0128, tangentMode=RCTM_User, time=1, value=0}
            ]
           }}}
         CurveIndex=$Particles.PartitionParticles.Normalized Index In Partition
         Scale Curve=&Lerp_Float(A=21, B=15, Alpha=$Particles.Waveform1)
        )
       )
       Y=0
       Z=0
      )
     )
     2=?
     3=?
     Index=$Particles.PartitionParticles.Partition
     Num=::NewEnumerator0
    )
    Skeletal Mesh=$System.Skeletal Mesh
    Use Sampled Mesh Attribute as Ribbon ID=false
   )
   &ScaleColor(
    Scale RGB=&Multiply_Vector(
     A=&MakeVectorFromLinearColorRGB(Color=&ColorFromCurve(
       ColorCurve=DataInterface() {propertyValues {
         AlphaCurve {keys [<at 0 1 RCIM_Cubic>]}
         BlueCurve {keys [<at 0.0229 0.0212 RCIM_Cubic>,<at 0.1897 0.2747>,<at 0.4637 0.2747>,<at 0.5356 0.2747>,<at 0.7777 0.2747>,<at 0.9715 0.0212 RCIM_Cubic>]}
         GreenCurve {keys [<at 0.0229 0.0529 RCIM_Cubic>,<at 0.1897 0.0438>,<at 0.4637 0>,<at 0.5356 0>,<at 0.7777 0.0595>,<at 0.9715 0.0529 RCIM_Cubic>]}
         RedCurve {keys [<at 0.0229 0 RCIM_Cubic>,<at 0.1897 0>,<at 0.4637 0.117>,<at 0.5356 0.117>,<at 0.7777 0>,<at 0.9715 0 RCIM_Cubic>]}
        }}
       CurveIndex=&Frac_Float(Float=&Add_Float(A=$Particles.PartitionParticles.Normalized Index In Partition, B=&Multiply_Float(A=$Engine.Time, B=0.5)))
      ))
     B=&VectorFromFloat(Value=0.3)
    )
    ScaleA=true
    ScaleRGB=true
   )
   &ScaleRibbonWidth(Ribbon Width Scale=&Multiply_Float(
     A=&MergeHelper_FloatSwitch(
      0=&Lerp_Float(A=0.1, B=1.25, Alpha=$Particles.Waveform2)
      1=&Lerp_Float(A=1.25, B=0.1, Alpha=$Particles.Waveform2)
      2=?
      3=?
      Index=$Particles.PartitionParticles.Partition
      Num=::NewEnumerator0
     )
     B=&FloatFromCurve(
      FloatCurve=DataInterface() {propertyValues {Curve {keys [
          {arriveTangent=4.1029, interpMode=RCIM_Cubic, leaveTangent=4.1029, tangentMode=RCTM_User, time=0, value=0.5}
          {arriveTangentWeight=0.2426, interpMode=RCIM_Cubic, leaveTangentWeight=0.082, tangentMode=RCTM_User, time=0.15, value=1}
          {arriveTangentWeight=0.1132, interpMode=RCIM_Cubic, leaveTangentWeight=0.2253, tangentMode=RCTM_User, time=0.2007, value=1}
          {arriveTangent=0.1507, interpMode=RCIM_Cubic, leaveTangent=0.675, tangentMode=RCTM_Break, time=0.5007, value=0.5}
          {arriveTangentWeight=0.2258, interpMode=RCIM_Cubic, leaveTangentWeight=0.1124, tangentMode=RCTM_User, time=0.8007, value=1}
          {arriveTangentWeight=0.135, interpMode=RCIM_Cubic, leaveTangentWeight=0.0442, tangentMode=RCTM_User, time=0.85, value=1}
          {arriveTangent=-8.6495, interpMode=RCIM_Cubic, leaveTangent=-8.6495, tangentMode=RCTM_User, time=0.9989, value=0.5}
         ]}}}
      CurveIndex=$Particles.PartitionParticles.Normalized Index In Partition
      Scale Curve=$System.Timeline.Float
     )
    ))
   /SetVariables(Particles.VectortoCamera:Vector3f=&ReturnCameraVector(Camera Vectors=::NewEnumerator4, Coordinate Space=::Simulation, Origin Position=$Particles.Position)) {isSetParameters=true, module=SetVariables_#6}
   /SetVariables(Particles.CameraOffsetRibbon:NiagaraFloat=&Lerp_Float(
     A=&MakeCustomFloatFromBool(Bool=$User.bIsFrontEndPreview, True Float=-30, False Float=-35)
     B=-15
     Alpha=&MakeCustomFloatFromBool(
      Bool=&SetBoolByFloatComparison(
       A=&OneMinusFloat(Float=&DotProduct(Vector A=$Particles.VectortoCamera, Vector B=<0 0 1>))
       B=0.5
      )
      True Float=&FloatFromCurve(
       FloatCurve=DataInterface() {propertyValues {Curve {keys [<at 0.55 0>,<at 0.95 1>]}}}
       CurveIndex=&Abs_Float(Float:NiagaraFloat=&DotProduct(Vector A=$Particles.VectortoCamera, Vector B=<0 0 1>))
      )
      False Float=0
     )
    )) {isSetParameters=true, module=SetVariables_#7}
   &InterpolateOverTime(
    Rate Of Change:NiagaraFloat=&FloatFromCurve(
     FloatCurve=DataInterface() {propertyValues {Curve {keys [<at 0 1 RCIM_Cubic>,<at 0.5 0.75 RCIM_Cubic>,<at 1 1 RCIM_Cubic>]}}}
     CurveIndex=$Particles.PartitionParticles.Normalized Index In Partition
     Scale Curve=&Multiply_Float(
      A=40
      B=&FloatFromCurve(
       FloatCurve=DataInterface() {propertyValues {Curve {keys [
           {interpMode=RCIM_Cubic, leaveTangentWeight=0.1728, tangentMode=RCTM_User, time=0.1, value=0.65}
           {
            arriveTangent=2.6077
            arriveTangentWeight=1.3568
            interpMode=RCIM_Cubic
            leaveTangent=-31.7719
            leaveTangentWeight=0.4846
            tangentMode=RCTM_Break
            time=0.45
            value=1
           }
          ]}}}
       CurveIndex=&Divide_Float(A=&VectorLength(Vector=$Engine.Owner.Velocity), B=300)
      )
     )
    )
    Target Position Value:NiagaraPosition=$Particles.Position
    Variable Type=::NewEnumerator6
   )
   /SetVariables(Particles.RibbonShape:NiagaraPosition=&AddVectorToPosition(
     Position=&AddVectorToPosition(
      Position=$Particles.InterpolateOverTime.InterpolateOverTime_Position.Moving Average
      Vector=&Multiply_Vector(A=$Particles.VectortoCamera, B=&VectorFromFloat(Value=$Particles.CameraOffsetRibbon))
     )
     Vector=&AddVectorToPosition(
      Position=&MergeHelper_VectorSwitch(
       0=&MakeVector(
        X=&FloatFromCurve(
         FloatCurve=DataInterface() {propertyValues {Curve {
            defaultValue=0
            keys [
             {arriveTangent=6.8692, interpMode=RCIM_Cubic, leaveTangent=6.8692, tangentMode=RCTM_User, time=0, value=0}
             {arriveTangent=0.1758, interpMode=RCIM_Cubic, leaveTangent=0.1758, time=0.1717, value=0.9182}
             {interpMode=RCIM_Cubic, tangentMode=RCTM_User, time=0.4953, value=0.0871}
             {arriveTangent=-0.1725, interpMode=RCIM_Cubic, leaveTangent=-0.1725, time=0.8022, value=0.9371}
             {arriveTangent=-7.5341, tangentMode=RCTM_User, time=1, value=0}
            ]
           }}}
         CurveIndex=$Particles.PartitionParticles.Normalized Index In Partition
         Scale Curve=&Multiply_Float(
          A=$Particles.CurveAdjust
          B=&Lerp_Float(
           A=7
           B=-7
           Alpha=&Waveform(
            Global Phase=&Frac_Float(Float=&Add_Float(A=$Particles.PartitionParticles.Normalized Index In Partition, B=&Multiply_Float(A=$Engine.Time, B=0.07)))
            "Waveform [1]"=::NewEnumerator2
            "[1] Frequency"=1
            "[1] Amplitude Min/Max"=<0 1>
           )
          )
         )
        )
        Y=&FloatFromCurve(
         FloatCurve=DataInterface() {propertyValues {Curve {
            defaultValue=0
            keys [
             {arriveTangent=6.8692, interpMode=RCIM_Cubic, leaveTangent=6.8692, tangentMode=RCTM_User, time=0, value=0}
             {arriveTangent=0.1758, interpMode=RCIM_Cubic, leaveTangent=0.1758, time=0.1717, value=0.9182}
             {interpMode=RCIM_Cubic, tangentMode=RCTM_User, time=0.4953, value=0.0871}
             {arriveTangent=-0.1725, interpMode=RCIM_Cubic, leaveTangent=-0.1725, time=0.8022, value=0.9371}
             {arriveTangent=-7.5341, tangentMode=RCTM_User, time=1, value=0}
            ]
           }}}
         CurveIndex=$Particles.PartitionParticles.Normalized Index In Partition
         Scale Curve=&Multiply_Float(
          A=$Particles.CurveAdjust
          B=&Lerp_Float(
           A=-10
           B=10
           Alpha=&Waveform(
            Global Phase=&Frac_Float(Float=&Add_Float(A=$Particles.PartitionParticles.Normalized Index In Partition, B=&Multiply_Float(A=$Engine.Time, B=0.05)))
            "Waveform [1]"=::NewEnumerator2
            "[1] Phase Scale"=1.5
            "[1] Frequency"=1
            "[1] Amplitude Min/Max"=<0 1>
           )
          )
         )
        )
        Z=&Multiply_Float(
         A=&FloatFromCurve(
          FloatCurve=DataInterface() {propertyValues {Curve {
             defaultValue=0
             keys [
              {arriveTangent=4.0936, interpMode=RCIM_Cubic, leaveTangent=4.0936, tangentMode=RCTM_User, time=0, value=0}
              {interpMode=RCIM_Cubic, tangentMode=RCTM_User, time=0.5, value=1}
              {arriveTangent=-3.0128, tangentMode=RCTM_User, time=1, value=0}
             ]
            }}}
          CurveIndex=$Particles.PartitionParticles.Normalized Index In Partition
          Scale Curve=&Lerp_Float(A=32, B=40, Alpha=$Particles.Waveform1)
         )
         B=$Particles.HeightAdjust
        )
       )
       1=&MakeVector(
        X=&FloatFromCurve(
         FloatCurve=DataInterface() {propertyValues {Curve {
            defaultValue=0
            keys [
             {arriveTangent=6.8692, interpMode=RCIM_Cubic, leaveTangent=6.8692, tangentMode=RCTM_User, time=0, value=0}
             {arriveTangent=0.1741, interpMode=RCIM_Cubic, leaveTangent=0.1741, time=0.1717, value=0.9182}
             {interpMode=RCIM_Cubic, tangentMode=RCTM_User, time=0.5, value=0.0871}
             {arriveTangent=-0.1741, interpMode=RCIM_Cubic, leaveTangent=-0.1741, time=0.8022, value=0.9371}
             {arriveTangent=-7.5341, tangentMode=RCTM_User, time=1, value=0}
            ]
           }}}
         CurveIndex=$Particles.PartitionParticles.Normalized Index In Partition
         Scale Curve=&Multiply_Float(
          A=$Particles.CurveAdjust
          B=&Lerp_Float(
           A=10
           B=-10
           Alpha=&Waveform(
            Global Phase=&Frac_Float(Float=&Add_Float(A=$Particles.PartitionParticles.Normalized Index In Partition, B=&Multiply_Float(A=$Engine.Time, B=0.045)))
            "Waveform [1]"=::NewEnumerator2
            "[1] Phase Scale"=1.5
            "[1] Frequency"=1
            "[1] Amplitude Min/Max"=<0 1>
           )
          )
         )
        )
        Y=&FloatFromCurve(
         FloatCurve=DataInterface() {propertyValues {Curve {
            defaultValue=0
            keys [
             {arriveTangent=6.8692, interpMode=RCIM_Cubic, leaveTangent=6.8692, tangentMode=RCTM_User, time=0, value=0}
             {arriveTangent=0.1741, interpMode=RCIM_Cubic, leaveTangent=0.1741, time=0.1717, value=0.9182}
             {interpMode=RCIM_Cubic, tangentMode=RCTM_User, time=0.5, value=0.0871}
             {arriveTangent=-0.1741, interpMode=RCIM_Cubic, leaveTangent=-0.1741, time=0.8022, value=0.9371}
             {arriveTangent=-7.5341, tangentMode=RCTM_User, time=1, value=0}
            ]
           }}}
         CurveIndex=$Particles.PartitionParticles.Normalized Index In Partition
         Scale Curve=&Multiply_Float(
          A=$Particles.CurveAdjust
          B=&Lerp_Float(
           A=-8
           B=8
           Alpha=&Waveform(
            Global Phase=&Frac_Float(Float=&Add_Float(A=$Particles.PartitionParticles.Normalized Index In Partition, B=&Multiply_Float(A=$Engine.Time, B=0.06)))
            "Waveform [1]"=::NewEnumerator2
            "[1] Frequency"=1
            "[1] Amplitude Min/Max"=<0 1>
           )
          )
         )
        )
        Z=&Multiply_Float(
         A=&FloatFromCurve(
          FloatCurve=DataInterface() {propertyValues {Curve {
             defaultValue=0
             keys [
              {arriveTangent=4.0936, interpMode=RCIM_Cubic, leaveTangent=4.0936, tangentMode=RCTM_User, time=0, value=0}
              {interpMode=RCIM_Cubic, tangentMode=RCTM_User, time=0.5, value=1}
              {arriveTangent=-3.0128, tangentMode=RCTM_User, time=1, value=0}
             ]
            }}}
          CurveIndex=$Particles.PartitionParticles.Normalized Index In Partition
          Scale Curve=&Lerp_Float(A=42, B=30, Alpha=$Particles.Waveform1)
         )
         B=$Particles.HeightAdjust
        )
       )
       2=?
       3=?
       Index=$Particles.PartitionParticles.Partition
       Num=::NewEnumerator0
      )
      Vector=&Add_Vector(
       A=&Multiply_Vector(
        A=&MergeHelper_VectorSwitch(
         0=&VectorFromFloat(Value=&Multiply_Float(
           A=&FloatFromCurve(
            FloatCurve=DataInterface() {propertyValues {Curve {
               defaultValue=1
               keys [
                {arriveTangent=3.4053, interpMode=RCIM_Cubic, leaveTangent=3.4053, tangentMode=RCTM_User, time=0, value=0}
                {arriveTangentWeight=0.08, interpMode=RCIM_Cubic, leaveTangentWeight=0.3468, tangentMode=RCTM_User, time=0.15, value=1}
                {arriveTangentWeight=0.2167, interpMode=RCIM_Cubic, leaveTangentWeight=0.0667, tangentMode=RCTM_User, time=0.85, value=1}
               ]
              }}}
            CurveIndex=$Particles.PartitionParticles.Normalized Index In Partition
            Scale Curve=$Particles.WidthAdjust
           )
           B=&Multiply_Float(
            A=&FloatFromCurve(
             FloatCurve=DataInterface() {propertyValues {Curve {
                defaultValue=0
                keys [
                 {interpMode=RCIM_Cubic, leaveTangentWeight=0.2361, tangentMode=RCTM_User, time=0.05, value=0}
                 {
                  arriveTangent=0.7804
                  arriveTangentWeight=0.2787
                  interpMode=RCIM_Cubic
                  leaveTangent=0.7804
                  leaveTangentWeight=0.0872
                  tangentMode=RCTM_User
                  time=0.7455
                  value=1
                 }
                 {arriveTangent=-11.0422, arriveTangentWeight=0.7206, interpMode=RCIM_Cubic, leaveTangent=-11.0422, tangentMode=RCTM_User, time=1, value=0.1}
                ]
               }}}
             CurveIndex=$Particles.PartitionParticles.Normalized Index In Partition
             Scale Curve=-29
            )
            B=1
           )
          ))
         1=&VectorFromFloat(Value=&Multiply_Float(
           A=&FloatFromCurve(
            FloatCurve=DataInterface() {propertyValues {Curve {
               defaultValue=1
               keys [
                {arriveTangent=3.4053, interpMode=RCIM_Cubic, leaveTangent=3.4053, tangentMode=RCTM_User, time=0, value=0}
                {arriveTangentWeight=0.08, interpMode=RCIM_Cubic, leaveTangentWeight=0.3468, tangentMode=RCTM_User, time=0.15, value=1}
                {arriveTangentWeight=0.2167, interpMode=RCIM_Cubic, leaveTangentWeight=0.0667, tangentMode=RCTM_User, time=0.85, value=1}
               ]
              }}}
            CurveIndex=$Particles.PartitionParticles.Normalized Index In Partition
            Scale Curve=$Particles.WidthAdjust
           )
           B=&Multiply_Float(
            A=&FloatFromCurve(
             FloatCurve=DataInterface() {propertyValues {Curve {
                defaultValue=0
                keys [
                 {interpMode=RCIM_Cubic, leaveTangentWeight=0.2633, tangentMode=RCTM_User, time=0.1, value=0}
                 {arriveTangentWeight=0.2787, interpMode=RCIM_Cubic, leaveTangentWeight=0.0872, tangentMode=RCTM_User, time=0.724, value=1}
                 {arriveTangent=-12.3048, arriveTangentWeight=0.7206, interpMode=RCIM_Cubic, leaveTangent=-12.3048, tangentMode=RCTM_User, time=1, value=0}
                ]
               }}}
             CurveIndex=$Particles.PartitionParticles.Normalized Index In Partition
             Scale Curve=25
            )
            B=1
           )
          ))
         2=?
         3=?
         Index=$Particles.PartitionParticles.Partition
         Num=::NewEnumerator0
        )
        B=$System.CameraRight_Vector
       )
       B=&Multiply_Vector(
        A=$System.CameraUp_Vector
        B=&VectorFromFloat(Value=&FloatFromCurve(
          FloatCurve=DataInterface() {propertyValues {Curve {
             defaultValue=1
             keys [
              {interpMode=RCIM_Cubic, tangentMode=RCTM_User, time=0, value=0.02}
              {interpMode=RCIM_Cubic, tangentMode=RCTM_User, time=0.5, value=1}
              {tangentMode=RCTM_User, time=1, value=0.02}
             ]
            }}}
          CurveIndex=$Particles.PartitionParticles.Normalized Index In Partition
          Scale Curve=&Lerp_Float(
           A=0
           B=18
           Alpha=&MakeCustomFloatFromBool(
            Bool=&SetBoolByFloatComparison(
             A=&OneMinusFloat(Float=&DotProduct(Vector A=$Particles.VectortoCamera, Vector B=<0 0 1>))
             B=0.5
            )
            True Float=&FloatFromCurve(
             FloatCurve=DataInterface() {propertyValues {Curve {keys [<at 0.5 0>,<at 1 1>]}}}
             CurveIndex=&Abs_Float(Float:NiagaraFloat=&DotProduct(Vector A=$Particles.VectortoCamera, Vector B=<0 0 1>))
             Scale Curve=2
            )
            False Float=0
           )
          )
         ))
       )
      )
     )
    )) {enabled=false, isSetParameters=true, module=SetVariables_#8}
   /SetVariables(Particles.RibbonShape:NiagaraPosition=&AddVectorToPosition(
     Position=&AddVectorToPosition(
      Position=$Particles.InterpolateOverTime.InterpolateOverTime_Position.Moving Average
      Vector=&Multiply_Vector(A=$Particles.VectortoCamera, B=&VectorFromFloat(Value=$Particles.CameraOffsetRibbon))
     )
     Vector=&AddVectorToPosition(
      Position=&MergeHelper_VectorSwitch(
       0=&MakeVector(
        X=&FloatFromCurve(
         FloatCurve=DataInterface() {propertyValues {Curve {
            defaultValue=0
            keys [
             {arriveTangent=6.8692, interpMode=RCIM_Cubic, leaveTangent=6.8692, tangentMode=RCTM_User, time=0, value=0}
             {arriveTangent=0.1758, interpMode=RCIM_Cubic, leaveTangent=0.1758, time=0.1717, value=0.9182}
             {interpMode=RCIM_Cubic, tangentMode=RCTM_User, time=0.4953, value=0.0871}
             {arriveTangent=-0.1725, interpMode=RCIM_Cubic, leaveTangent=-0.1725, time=0.8022, value=0.9371}
             {arriveTangent=-7.5341, tangentMode=RCTM_User, time=1, value=0}
            ]
           }}}
         CurveIndex=$Particles.PartitionParticles.Normalized Index In Partition
         Scale Curve=&Multiply_Float(
          A=$Particles.CurveAdjust
          B=&Lerp_Float(
           A=7
           B=-7
           Alpha=&Waveform(
            Global Phase=&Frac_Float(Float=&Add_Float(A=$Particles.PartitionParticles.Normalized Index In Partition, B=&Multiply_Float(A=$Engine.Time, B=0.07)))
            "Waveform [1]"=::NewEnumerator2
            "[1] Frequency"=1
            "[1] Amplitude Min/Max"=<0 1>
           )
          )
         )
        )
        Y=&FloatFromCurve(
         FloatCurve=DataInterface() {propertyValues {Curve {
            defaultValue=0
            keys [
             {arriveTangent=6.8692, interpMode=RCIM_Cubic, leaveTangent=6.8692, tangentMode=RCTM_User, time=0, value=0}
             {arriveTangent=0.1758, interpMode=RCIM_Cubic, leaveTangent=0.1758, time=0.1717, value=0.9182}
             {interpMode=RCIM_Cubic, tangentMode=RCTM_User, time=0.4953, value=0.0871}
             {arriveTangent=-0.1725, interpMode=RCIM_Cubic, leaveTangent=-0.1725, time=0.8022, value=0.9371}
             {arriveTangent=-7.5341, tangentMode=RCTM_User, time=1, value=0}
            ]
           }}}
         CurveIndex=$Particles.PartitionParticles.Normalized Index In Partition
         Scale Curve=&Multiply_Float(
          A=$Particles.CurveAdjust
          B=&Lerp_Float(
           A=-10
           B=10
           Alpha=&Waveform(
            Global Phase=&Frac_Float(Float=&Add_Float(A=$Particles.PartitionParticles.Normalized Index In Partition, B=&Multiply_Float(A=$Engine.Time, B=0.05)))
            "Waveform [1]"=::NewEnumerator2
            "[1] Phase Scale"=1.5
            "[1] Frequency"=1
            "[1] Amplitude Min/Max"=<0 1>
           )
          )
         )
        )
        Z=&Multiply_Float(
         A=&FloatFromCurve(
          FloatCurve=DataInterface() {propertyValues {Curve {
             defaultValue=0
             keys [
              {arriveTangent=4.0936, interpMode=RCIM_Cubic, leaveTangent=4.0936, tangentMode=RCTM_User, time=0, value=0}
              {interpMode=RCIM_Cubic, tangentMode=RCTM_User, time=0.5, value=1}
              {arriveTangent=-3.0128, tangentMode=RCTM_User, time=1, value=0}
             ]
            }}}
          CurveIndex=$Particles.PartitionParticles.Normalized Index In Partition
          Scale Curve=&Lerp_Float(A=16, B=20, Alpha=$Particles.Waveform1)
         )
         B=$Particles.HeightAdjust
        )
       )
       1=&MakeVector(
        X=&FloatFromCurve(
         FloatCurve=DataInterface() {propertyValues {Curve {
            defaultValue=0
            keys [
             {arriveTangent=6.8692, interpMode=RCIM_Cubic, leaveTangent=6.8692, tangentMode=RCTM_User, time=0, value=0}
             {arriveTangent=0.1741, interpMode=RCIM_Cubic, leaveTangent=0.1741, time=0.1717, value=0.9182}
             {interpMode=RCIM_Cubic, tangentMode=RCTM_User, time=0.5, value=0.0871}
             {arriveTangent=-0.1741, interpMode=RCIM_Cubic, leaveTangent=-0.1741, time=0.8022, value=0.9371}
             {arriveTangent=-7.5341, tangentMode=RCTM_User, time=1, value=0}
            ]
           }}}
         CurveIndex=$Particles.PartitionParticles.Normalized Index In Partition
         Scale Curve=&Multiply_Float(
          A=$Particles.CurveAdjust
          B=&Lerp_Float(
           A=10
           B=-10
           Alpha=&Waveform(
            Global Phase=&Frac_Float(Float=&Add_Float(A=$Particles.PartitionParticles.Normalized Index In Partition, B=&Multiply_Float(A=$Engine.Time, B=0.045)))
            "Waveform [1]"=::NewEnumerator2
            "[1] Phase Scale"=1.5
            "[1] Frequency"=1
            "[1] Amplitude Min/Max"=<0 1>
           )
          )
         )
        )
        Y=&FloatFromCurve(
         FloatCurve=DataInterface() {propertyValues {Curve {
            defaultValue=0
            keys [
             {arriveTangent=6.8692, interpMode=RCIM_Cubic, leaveTangent=6.8692, tangentMode=RCTM_User, time=0, value=0}
             {arriveTangent=0.1741, interpMode=RCIM_Cubic, leaveTangent=0.1741, time=0.1717, value=0.9182}
             {interpMode=RCIM_Cubic, tangentMode=RCTM_User, time=0.5, value=0.0871}
             {arriveTangent=-0.1741, interpMode=RCIM_Cubic, leaveTangent=-0.1741, time=0.8022, value=0.9371}
             {arriveTangent=-7.5341, tangentMode=RCTM_User, time=1, value=0}
            ]
           }}}
         CurveIndex=$Particles.PartitionParticles.Normalized Index In Partition
         Scale Curve=&Multiply_Float(
          A=$Particles.CurveAdjust
          B=&Lerp_Float(
           A=-8
           B=8
           Alpha=&Waveform(
            Global Phase=&Frac_Float(Float=&Add_Float(A=$Particles.PartitionParticles.Normalized Index In Partition, B=&Multiply_Float(A=$Engine.Time, B=0.06)))
            "Waveform [1]"=::NewEnumerator2
            "[1] Frequency"=1
            "[1] Amplitude Min/Max"=<0 1>
           )
          )
         )
        )
        Z=&Multiply_Float(
         A=&FloatFromCurve(
          FloatCurve=DataInterface() {propertyValues {Curve {
             defaultValue=0
             keys [
              {arriveTangent=4.0936, interpMode=RCIM_Cubic, leaveTangent=4.0936, tangentMode=RCTM_User, time=0, value=0}
              {interpMode=RCIM_Cubic, tangentMode=RCTM_User, time=0.5, value=1}
              {arriveTangent=-3.0128, tangentMode=RCTM_User, time=1, value=0}
             ]
            }}}
          CurveIndex=$Particles.PartitionParticles.Normalized Index In Partition
          Scale Curve=&Lerp_Float(A=21, B=15, Alpha=$Particles.Waveform1)
         )
         B=$Particles.HeightAdjust
        )
       )
       2=?
       3=?
       Index=$Particles.PartitionParticles.Partition
       Num=::NewEnumerator0
      )
      Vector=&Add_Vector(
       A=&Multiply_Vector(
        A=&MergeHelper_VectorSwitch(
         0=&VectorFromFloat(Value=&Multiply_Float(
           A=&FloatFromCurve(
            FloatCurve=DataInterface() {propertyValues {Curve {
               defaultValue=1
               keys [
                {arriveTangent=3.4053, interpMode=RCIM_Cubic, leaveTangent=3.4053, tangentMode=RCTM_User, time=0, value=0}
                {arriveTangentWeight=0.08, interpMode=RCIM_Cubic, leaveTangentWeight=0.3468, tangentMode=RCTM_User, time=0.15, value=1}
                {arriveTangentWeight=0.2167, interpMode=RCIM_Cubic, leaveTangentWeight=0.0667, tangentMode=RCTM_User, time=0.85, value=1}
               ]
              }}}
            CurveIndex=$Particles.PartitionParticles.Normalized Index In Partition
            Scale Curve=$Particles.WidthAdjust
           )
           B=&Multiply_Float(
            A=&FloatFromCurve(
             FloatCurve=DataInterface() {propertyValues {Curve {
                defaultValue=0
                keys [
                 {interpMode=RCIM_Cubic, leaveTangentWeight=0.2361, tangentMode=RCTM_User, time=0.05, value=0}
                 {
                  arriveTangent=0.7804
                  arriveTangentWeight=0.2787
                  interpMode=RCIM_Cubic
                  leaveTangent=0.7804
                  leaveTangentWeight=0.0872
                  tangentMode=RCTM_User
                  time=0.7455
                  value=1
                 }
                 {arriveTangent=-11.0422, arriveTangentWeight=0.7206, interpMode=RCIM_Cubic, leaveTangent=-11.0422, tangentMode=RCTM_User, time=1, value=0.1}
                ]
               }}}
             CurveIndex=$Particles.PartitionParticles.Normalized Index In Partition
             Scale Curve=-29
            )
            B=1
           )
          ))
         1=&VectorFromFloat(Value=&Multiply_Float(
           A=&FloatFromCurve(
            FloatCurve=DataInterface() {propertyValues {Curve {
               defaultValue=1
               keys [
                {arriveTangent=3.4053, interpMode=RCIM_Cubic, leaveTangent=3.4053, tangentMode=RCTM_User, time=0, value=0}
                {arriveTangentWeight=0.08, interpMode=RCIM_Cubic, leaveTangentWeight=0.3468, tangentMode=RCTM_User, time=0.15, value=1}
                {arriveTangentWeight=0.2167, interpMode=RCIM_Cubic, leaveTangentWeight=0.0667, tangentMode=RCTM_User, time=0.85, value=1}
               ]
              }}}
            CurveIndex=$Particles.PartitionParticles.Normalized Index In Partition
            Scale Curve=$Particles.WidthAdjust
           )
           B=&Multiply_Float(
            A=&FloatFromCurve(
             FloatCurve=DataInterface() {propertyValues {Curve {
                defaultValue=0
                keys [
                 {interpMode=RCIM_Cubic, leaveTangentWeight=0.2633, tangentMode=RCTM_User, time=0.1, value=0}
                 {arriveTangentWeight=0.2787, interpMode=RCIM_Cubic, leaveTangentWeight=0.0872, tangentMode=RCTM_User, time=0.724, value=1}
                 {arriveTangent=-12.3048, arriveTangentWeight=0.7206, interpMode=RCIM_Cubic, leaveTangent=-12.3048, tangentMode=RCTM_User, time=1, value=0}
                ]
               }}}
             CurveIndex=$Particles.PartitionParticles.Normalized Index In Partition
             Scale Curve=25
            )
            B=1
           )
          ))
         2=?
         3=?
         Index=$Particles.PartitionParticles.Partition
         Num=::NewEnumerator0
        )
        B=$System.CameraRight_Vector
       )
       B=&Multiply_Vector(
        A=$System.CameraUp_Vector
        B=&VectorFromFloat(Value=&FloatFromCurve(
          FloatCurve=DataInterface() {propertyValues {Curve {
             defaultValue=1
             keys [
              {interpMode=RCIM_Cubic, tangentMode=RCTM_User, time=0, value=0.02}
              {interpMode=RCIM_Cubic, tangentMode=RCTM_User, time=0.5, value=1}
              {tangentMode=RCTM_User, time=1, value=0.02}
             ]
            }}}
          CurveIndex=$Particles.PartitionParticles.Normalized Index In Partition
          Scale Curve=&Lerp_Float(
           A=0
           B=18
           Alpha=&MakeCustomFloatFromBool(
            Bool=&SetBoolByFloatComparison(
             A=&OneMinusFloat(Float=&DotProduct(Vector A=$Particles.VectortoCamera, Vector B=<0 0 1>))
             B=0.5
            )
            True Float=&FloatFromCurve(
             FloatCurve=DataInterface() {propertyValues {Curve {keys [<at 0.5 0>,<at 1 1>]}}}
             CurveIndex=&Abs_Float(Float:NiagaraFloat=&DotProduct(Vector A=$Particles.VectortoCamera, Vector B=<0 0 1>))
             Scale Curve=2
            )
            False Float=0
           )
          )
         ))
       )
      )
     )
    )) {isSetParameters=true, module=SetVariables_#9}
  ]
  rendererClasses [&NiagaraRibbonRendererProperties]
  renderers [{rendererClass=&NiagaraRibbonRendererProperties}]
 }
}
summary {
 systemName=NS_Aura_ShadedExpanse
 emitters [
  {emitterName=Decal_GlowRays, rendererClasses [&NiagaraSpriteRendererProperties]}
  {emitterName=Orbit_Blobs, rendererClasses [&NiagaraMeshRendererProperties]}
  {emitterName=LiquidMesh, rendererClasses [&NiagaraMeshRendererProperties]}
  {bEnabled=false, emitterName=RibbonWorldZOffset, rendererClasses [&NiagaraRibbonRendererProperties]}
  {emitterName=Camera_Oriented_Ribbon001, rendererClasses [&NiagaraRibbonRendererProperties]}
  {bEnabled=false, emitterName=Ribbon_HybridOffset, rendererClasses [&NiagaraRibbonRendererProperties]}
 ]
 userVariables [
  User.bIsFrontEndPreview {type=NiagaraBool}
  User.IsFrontEnd {type=NiagaraBool}
  User.IsLargeBody {type=NiagaraBool}
 ]
}
