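# Start burst Niagara system for the Shaded Expanse aura cosmetic, placing short-lived sprite effects on the character's skeletal mesh.
# CenterFlare and Corona_Glow_Combined make a purple-blue flash and glows, partly dimmed by time of day.
# Goop and Drops share identical spawn-group-switched logic with different bursts enabled: large splats versus small speed-stretched droplets.
# Both launch outward from the sampled mesh position, fall under gravity, and enlarge for large bodies.
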NiagaraSystem /SparksCosmetics/Items/Auras/Assets/NS_Aura_ShadedExpanse/NS_Aura_ShadedExpanse_Start
dependencies {
 usedDataInterfaces [&NiagaraDataInterfaceCurve, &NiagaraDataInterfaceColorCurve, &NiagaraDataInterfaceSkeletalMesh]
 usedDynamicInputs []
 usedModules [
  &self:SystemSpawnScript.SystemScriptSource.SystemScriptGraph.NiagaraNodeAssignment_0.SetVariables_#1
  &SystemState
 ]
}
emitters {
 CenterFlare {
  emitterUpdateScript [
   &EmitterState(Life Cycle Mode=::NewEnumerator1, Loop Behavior=::NewEnumerator1, Loop Duration=1, UseLoopDelay=false)
   &SpawnBurst_Instantaneous(Spawn Count=1, Use Spawn Probability=false)
  ]
  particleSpawnScript [
   &InitializeParticle(
    Color=<rgba 0.5314 0.2216 1.25 1>
    Lifetime=0.25
    Sprite Rotation Mode=::NewEnumerator2
    Sprite Size Mode=::NewEnumerator3
    Uniform Sprite Size=400
    UsePositionOffset=false
    Write Material Random=true
   )
   &DynamicMaterialParameters(
    Index 0 Param 1=&RandomRangeFloat(Minimum=0.4, Maximum=0.6)
    Index 0 Param 2=1
    Index 0 Param 3=1
    Index 0 Param 4=1
    Module.Param0WriteEnabled=true
    Module.Param1WriteEnabled=true
    Module.Param2WriteEnabled=true
    Module.Param3WriteEnabled=true
   )
   &SystemLocation(MaskSpawnGroup=false, Offset=<0 0 120>)
   /SetVariables(Particles.Sharp Flare Camera Offset=40) {isSetParameters=true}
   &CameraOffset(Camera Offset Amount=-30) {module=CameraOffset001}
  ]
  particleUpdateScript [
   &ParticleState()
   &SolveForcesAndVelocity(Clamp Velocity=false, Limit Acceleration=false)
   &SkeletalMeshLocation(Bone Sampling Mode=::NewEnumerator4, MaskSpawnGroup=false, Skeletal Mesh=$System.Skeletal Mesh) {module=SkeletalMeshLocation001}
   &ScaleColor(
    Scale Alpha=&Multiply_Float(
     A=&FNTimeOfDayScalar(Night Scale=1, Morning Scale=1, Evening Scale=1)
     B=&FloatFromCurve(FloatCurve=DataInterface() {propertyValues {Curve {keys [
          {arriveTangent=8.1746, interpMode=RCIM_Cubic, leaveTangent=8.1746, tangentMode=RCTM_User, time=0, value=0}
          {arriveTangentWeight=0.1067, interpMode=RCIM_Cubic, leaveTangentWeight=0.3202, tangentMode=RCTM_User, time=0.2, value=1}
          {arriveTangentWeight=0.3202, interpMode=RCIM_Cubic, leaveTangentWeight=0.1067, tangentMode=RCTM_User, time=0.5, value=1}
          <at 1 0 RCIM_Cubic>
         ]}}})
    )
    Scale RGB=&Multiply_VectorByFloat(Vector=<1 1 1>, Float=1)
    ScaleA=true
    ScaleRGB=true
   )
   &ScaleSpriteSize(Scale Factor=&Vector2DFromFloat(Value=&FloatFromCurve(FloatCurve=DataInterface() {propertyValues {Curve {keys [<at 0 1.5 RCIM_Cubic>,<at 1 0>]}}})))
   /SetVariables(Particles.Flash color:LinearColor=&MakeLinearColorFromVectorAndFloat(
     "Vector (RGB)"=&MakeVectorFromLinearColorRGB(Color=<rgba 0.1403 0.3856 1 1>)
     "Float (Alpha)"=&Multiply_Float(
      A=&Multiply_Float(
       A=&FNTimeOfDayScalar(Night Scale=1, Morning Scale=1, Evening Scale=1)
       B=&FloatFromCurve(FloatCurve=DataInterface() {propertyValues {Curve {keys [
            {arriveTangent=14.8591, interpMode=RCIM_Cubic, leaveTangent=14.8591, tangentMode=RCTM_User, time=0, value=0}
            {arriveTangentWeight=0.093, interpMode=RCIM_Cubic, leaveTangentWeight=0.2791, tangentMode=RCTM_User, time=0.1, value=1}
            {arriveTangentWeight=0.2791, interpMode=RCIM_Cubic, leaveTangentWeight=0.093, tangentMode=RCTM_User, time=0.5, value=1}
            <at 0.7 0 RCIM_Cubic>
           ]}}})
      )
      B=1
     )
    )) {isSetParameters=true}
   /SetVariables(Particles.Flash size:Vector2f=&Multiply_Vector2D_ByFloat(
     Float=&Multiply_Float(
      A=375
      B=&FloatFromCurve(FloatCurve=DataInterface() {propertyValues {Curve {keys [
           {arriveTangent=16.6356, interpMode=RCIM_Cubic, leaveTangent=16.6356, tangentMode=RCTM_User, time=0, value=0}
           {arriveTangent=-0.15, interpMode=RCIM_Cubic, leaveTangent=-0.15, tangentMode=RCTM_User, time=0.1378, value=1}
           {arriveTangent=-0.0405, interpMode=RCIM_Cubic, leaveTangent=-0.0405, tangentMode=RCTM_User, time=0.62, value=0.25}
          ]}}})
     )
     Vector2D=<1 1>
    )) {isSetParameters=true}
  ]
  rendererClasses [&NiagaraSpriteRendererProperties]
  renderers [
   {rendererClass=&NiagaraSpriteRendererProperties}
   {rendererClass=&NiagaraSpriteRendererProperties, rendererIndex=1}
  ]
 }
 Corona_Glow_Combined {
  emitterUpdateScript [
   &EmitterState(Life Cycle Mode=::NewEnumerator1, Loop Behavior=::NewEnumerator1, Loop Duration=1, UseLoopDelay=false)
   &SpawnBurst_Instantaneous(Spawn Count=1, Use Spawn Probability=false)
   &SpawnBurst_Instantaneous(Spawn Count=1, Spawn Group=1, Use Spawn Probability=false) {module=SpawnBurst_Instantaneous001}
  ]
  particleSpawnScript [
   /SetVariables(Particles.VisibilityTag=&SpawnGroup()) {isSetParameters=true}
   &InitializeParticle(
    Color=&MergeHelper_ColorSwitch(0=<rgba 0.2841 0.2769 1 1>, 1=<rgba 0.6528 0.1719 2 1>, 2=?, 3=?, Index=&SpawnGroup(), Num=::NewEnumerator0)
    Lifetime=&MergeHelper_FloatSwitch(0=0.3, 1=0.4, 2=?, 3=?, Index=&SpawnGroup(), Num=::NewEnumerator0)
    Sprite Rotation Mode=::NewEnumerator2
    Sprite Size Mode=::NewEnumerator3
    Uniform Sprite Size=&MergeHelper_FloatSwitch(0=400, 1=150, 2=?, 3=?, Index=&SpawnGroup(), Num=::NewEnumerator0)
    UsePositionOffset=false
    Write Material Random=true
   )
   &CameraOffset(Camera Offset Amount=&MergeHelper_FloatSwitch(0=-30, 1=-10, 2=?, 3=?, Index=&SpawnGroup(), Num=::NewEnumerator0))
   &SystemLocation(
    MaskSpawnGroup=false
    Offset=&MergeHelper_VectorSwitch(0=<0 0 120>, 1=<0 0 0>, 2=?, 3=?, Index=&SpawnGroup(), Num=::NewEnumerator0)
   )
  ]
  particleUpdateScript [
   &ParticleState()
   &SkeletalMeshLocation(Bone Sampling Mode=::NewEnumerator4, MaskSpawnGroup=false, Skeletal Mesh=$System.Skeletal Mesh) {module=SkeletalMeshLocation001}
   &ScaleColor(
    Scale Alpha=&MergeHelper_FloatSwitch(
     0=&Multiply_Float(
      A=&FloatFromCurve(
       FloatCurve=DataInterface() {propertyValues {Curve {keys [
           <at 0 0.85 RCIM_Cubic>
           {arriveTangent=0.3, interpMode=RCIM_Cubic, leaveTangent=0.3, time=0.2009, value=0.92}
           <at 0.5 1 RCIM_Cubic>
           {arriveTangent=-0.3, interpMode=RCIM_Cubic, leaveTangent=-0.3, time=0.8028, value=0.92}
           <at 1 0.85 RCIM_Cubic>
          ]}}}
       CurveIndex=Linked() {linkedVariable {name=NPC.FortniteNPC.TimeOfDay_0-1, type=NiagaraFloat}}
      )
      B=&FloatFromCurve(FloatCurve=DataInterface() {propertyValues {Curve {keys [
           {arriveTangent=9.6062, interpMode=RCIM_Cubic, leaveTangent=9.6062, tangentMode=RCTM_User, time=0, value=0}
           {interpMode=RCIM_Cubic, tangentMode=RCTM_User, time=0.2, value=1}
           {interpMode=RCIM_Cubic, tangentMode=RCTM_User, time=1, value=0}
          ]}}})
     )
     1=&FloatFromCurve(
      FloatCurve=DataInterface() {propertyValues {Curve {keys [
          {interpMode=RCIM_Cubic, tangentMode=RCTM_User, time=0, value=0}
          {interpMode=RCIM_Cubic, tangentMode=RCTM_User, time=0.2, value=1}
          {interpMode=RCIM_Cubic, tangentMode=RCTM_User, time=0.5, value=1}
          {tangentMode=RCTM_User, time=1, value=0}
         ]}}}
      Scale Curve=2
     )
     2=?
     3=?
     Index=&SpawnGroup()
     Num=::NewEnumerator0
    )
    Scale RGB=&MergeHelper_VectorSwitch(
     0=&Multiply_VectorByFloat(
      Vector=<1 1 1>
      Float=&FloatFromCurve(FloatCurve=DataInterface() {propertyValues {Curve {keys [
           {arriveTangent=-0.8944, interpMode=RCIM_Cubic, leaveTangent=-0.8944, tangentMode=RCTM_User, time=0, value=3}
           {arriveTangent=0.7137, interpMode=RCIM_Cubic, leaveTangent=0.7137, tangentMode=RCTM_User, time=1, value=1}
          ]}}})
     )
     1=&MakeVectorFromLinearColorRGB(Color=&ColorFromCurve(ColorCurve=DataInterface() {propertyValues {
         AlphaCurve {keys [<at 0 1 RCIM_Cubic>]}
         BlueCurve {keys [<at 0.1905 1 RCIM_Cubic>,<at 1 0.7578 RCIM_Cubic>]}
         GreenCurve {keys [<at 0.1905 1 RCIM_Cubic>,<at 1 0.1479 RCIM_Cubic>]}
         RedCurve {keys [<at 0.1905 1 RCIM_Cubic>,<at 1 0 RCIM_Cubic>]}
        }}))
     2=?
     3=?
     Index=&SpawnGroup()
     Num=::NewEnumerator0
    )
    ScaleA=true
    ScaleRGB=true
   )
   &ScaleSpriteSize(Scale Factor=&MergeHelper_Vector2Switch(
     0=&Vector2DFromFloat(Value=&FloatFromCurve(FloatCurve=DataInterface() {propertyValues {Curve {keys [
           {arriveTangent=2.1722, interpMode=RCIM_Cubic, leaveTangent=2.1722, tangentMode=RCTM_User, time=0, value=0}
           {arriveTangent=0.0224, interpMode=RCIM_Cubic, leaveTangent=0.0224, tangentMode=RCTM_User, time=1, value=1}
          ]}}}))
     1=<1 1>
     2=?
     3=?
     4=?
     Index=&SpawnGroup()
     Num=::NewEnumerator0
    ))
   &DynamicMaterialParameters(
    Index 0 Param 1=0.75
    Index 0 Param 2=2
    Index 0 Param 3=1
    Module.Index0Float0Write=true
    Module.Index0Float1Write=true
    Module.Index0Float2Write=true
    Module.Index0Float3Write=true
   ) {module=DynamicMaterialParameters001}
   /SetVariables(Particles.GlowDynamicParameter:Vector4f=<3 1 1 1>) {isSetParameters=true}
  ]
  rendererClasses [&NiagaraSpriteRendererProperties]
  renderers [
   {rendererClass=&NiagaraSpriteRendererProperties}
   {rendererClass=&NiagaraSpriteRendererProperties, rendererIndex=1}
  ]
 }
 Drops {
  emitterUpdateScript [
   &EmitterState(Life Cycle Mode=::NewEnumerator1, Loop Behavior=::NewEnumerator1, Loop Duration=2, UseLoopDelay=false)
   &SpawnBurst_Instantaneous(Use Spawn Probability=false, UseLoopCountLimit=false) {enabled=false}
   &SpawnBurst_Instantaneous(Spawn Group=1, Use Spawn Probability=false, UseLoopCountLimit=false) {enabled=false, module=SpawnBurst_Instantaneous001}
   &SpawnBurst_Instantaneous(
    Spawn Count=&RandomRangeInt(Minimum=25, Maximum=30)
    Spawn Group=2
    Use Spawn Probability=false
    UseLoopCountLimit=false
   ) {module=SpawnBurst_Instantaneous002}
  ]
  particleSpawnScript [
   /SetVariables(Particles.VisibilityTag=&SpawnGroup()) {isSetParameters=true, module=SetVariables_#2}
   &InitializeParticle(
    Lifetime=&MergeHelper_FloatSwitch(
     0=0.4
     1=0.5
     2=&RandomRangeFloat(Minimum=0.4, Maximum=0.7)
     3=?
     Index=&SpawnGroup()
     Num=::NewEnumerator1
    )
    Sprite Rotation Mode=::NewEnumerator2
    Sprite Size Mode=::NewEnumerator3
    Uniform Sprite Size=&Multiply_Float(
     A=&MergeHelper_FloatSwitch(
      0=75
      1=95
      2=&RandomRangeFloat(Minimum=6, Maximum=8)
      3=?
      Index=&SpawnGroup()
      Num=::NewEnumerator1
     )
     B=&MakeCustomFloatFromBool(Bool=$User.IsLargeBody, True Float=1.1, False Float=1)
    )
    UsePositionOffset=false
    Write Material Random=true
   )
   &ShapeLocation(
    Apply To Particle Position=::NewEnumerator1
    Hemisphere Distribution=&MergeHelper_Vector2Switch(0=<0 0>, 1=<0 0>, 2=<1 1>, 3=?, 4=?, Index=&SpawnGroup(), Num=::NewEnumerator1)
    MaskSpawnGroup=false
    Override Seed=false
    Shape Origin=&VectorFromFloat(Value=0)
    Sphere Radius=&MergeHelper_FloatSwitch(0=0, 1=0, 2=15, 3=?, Index=&SpawnGroup(), Num=::NewEnumerator1)
    Sphere Surface Distribution=0.4524
    UseOwnerScale=true
   )
   &SkeletalMeshLocation(
    Bone Sampling Mode=::NewEnumerator4
    MaskSpawnGroup=false
    Orientation Sampling=::NewEnumerator1
    Position Sampling=::NewEnumerator1
    Sampled Position Offset=&AddVectorToPosition(Position=$Particles.ShapeLocation.ShapePosition, Vector=<15 0 0>)
    Skeletal Mesh=$System.Skeletal Mesh
   )
   &AddVelocity(
    Module.Use Falloff Distance=false
    Origin Offset=&MergeHelper_VectorSwitch(0=<0 0 0>, 1=<0 0 0>, 2=<0 0 5>, 3=?, Index=&SpawnGroup(), Num=::NewEnumerator1)
    Override Seed=false
    Velocity Mode=::NewEnumerator1
    Velocity Origin=$Output.SkeletalMeshLocation.SampledPositionWithoutOffset
    Velocity Speed=&MergeHelper_FloatSwitch(
     0=50
     1=25
     2=&RandomRangeFloat(Minimum=100, Maximum=300)
     3=?
     Index=&SpawnGroup()
     Num=::NewEnumerator1
    )
   ) {module=AddVelocity001}
   &CameraOffset(Camera Offset Amount=-25)
   /SetVariables(Particles.Frame:NiagaraFloat=&MergeHelper_FloatSwitch(
     0=0
     1=&MakeFloatFromInt(Int=&RandomRangeInt(Minimum=0, Maximum=3))
     2=&MakeFloatFromInt(Int=&RandomRangeInt(Minimum=0, Maximum=3))
     3=?
     Index=&SpawnGroup()
     Num=::NewEnumerator1
    )) {isSetParameters=true, module=SetVariables_#3}
  ]
  particleUpdateScript [
   &ParticleState()
   &ScaleSpriteSize(
    Scale Sprite Size Mode=::NewEnumerator0
    Uniform Scale Factor=&MergeHelper_FloatSwitch(
     0=&FloatFromCurve(FloatCurve=DataInterface() {propertyValues {Curve {keys [
          {interpMode=RCIM_Cubic, leaveTangent=2.7083, tangentMode=RCTM_Break, time=0, value=0}
          {arriveTangent=1.1018, interpMode=RCIM_Cubic, leaveTangent=1.1018, time=0.4356, value=0.8}
          {interpMode=RCIM_Cubic, leaveTangent=0.9481, time=0.9983, value=1.1}
         ]}}})
     1=&FloatFromCurve(FloatCurve=DataInterface() {propertyValues {Curve {keys [
          {interpMode=RCIM_Cubic, leaveTangent=2.7083, tangentMode=RCTM_Break, time=0, value=0}
          {arriveTangent=1.524, interpMode=RCIM_Cubic, leaveTangent=1.524, time=0.4356, value=0.8}
          {arriveTangent=0.5331, interpMode=RCIM_Cubic, leaveTangent=0.5331, time=0.647, value=0.986}
          {interpMode=RCIM_Cubic, leaveTangent=0.9481, time=0.9983, value=1.1}
         ]}}})
     2=&FloatFromCurve(FloatCurve=DataInterface() {propertyValues {Curve {
         defaultValue=1
         keys [
          {arriveTangent=4.7388, interpMode=RCIM_Cubic, leaveTangent=4.7388, tangentMode=RCTM_User, time=0, value=0}
          {arriveTangentWeight=0.1274, interpMode=RCIM_Cubic, leaveTangentWeight=0.7778, tangentMode=RCTM_User, time=0.2, value=1}
          {arriveTangent=-1.9979, interpMode=RCIM_Cubic, leaveTangent=-1.9979, tangentMode=RCTM_User, time=1, value=0}
         ]
        }}})
     3=?
     Index=&SpawnGroup()
     Num=::NewEnumerator1
    )
   )
   &GravityForce(Gravity=&MergeHelper_VectorSwitch(
     0=<0 0 -200>
     1=<0 0 -200>
     2=&RandomRangeVector(Minimum=<0 0 -350>, Maximum=<0 0 -500>)
     3=?
     Index=&SpawnGroup()
     Num=::NewEnumerator1
    ))
   &DynamicMaterialParameters(
    Index 0 Param 1=&MergeHelper_FloatSwitch(
     0=&RandomRangeFloat(Minimum=0, Maximum=1)
     1=0
     2=0
     3=?
     Index=&SpawnGroup()
     Num=::NewEnumerator1
    )
    Index 0 Param 3=&FloatFromCurve(FloatCurve=DataInterface() {propertyValues {Curve {keys [
         {interpMode=RCIM_Cubic, tangentMode=RCTM_Break, time=0.2, value=-0.6}
         {arriveTangent=-0.3316, interpMode=RCIM_Cubic, tangentMode=RCTM_Break, time=1, value=-1.5}
        ]}}})
    Index 0 Param 4=&FloatFromCurve(FloatCurve=DataInterface() {propertyValues {Curve {keys [
         {arriveTangent=2.6651, interpMode=RCIM_Cubic, leaveTangent=2.6651, tangentMode=RCTM_User, time=0, value=0.1}
         {arriveTangent=3.4594, interpMode=RCIM_Cubic, leaveTangent=3.4594, tangentMode=RCTM_User, time=1, value=3}
        ]}}})
    Module.Index0Float0Write=true
    Module.Index0Float1Write=true
    Module.Index0Float2Write=true
    Module.Index0Float3Write=true
   ) {module=DynamicMaterialParameters001}
   &Drag(Use Linear Drag=true, Use Rotational Drag=true)
   &SolveForcesAndVelocity(Clamp Velocity=false, Limit Acceleration=false)
   &ScaleSpriteSizeBySpeed(
    Max Scale Factor=&MergeHelper_Vector2Switch(0=<1 1>, 1=<1 1>, 2=<0.5 3>, 3=?, 4=?, Index=&SpawnGroup(), Num=::NewEnumerator1)
    Min Scale Factor=&MergeHelper_Vector2Switch(0=<1 1>, 1=<1 1>, 2=<0.9 1.2>, 3=?, 4=?, Index=&SpawnGroup(), Num=::NewEnumerator1)
   )
  ]
  rendererClasses [&NiagaraSpriteRendererProperties]
  renderers [
   {rendererClass=&NiagaraSpriteRendererProperties}
   {rendererClass=&NiagaraSpriteRendererProperties, rendererIndex=1}
   {rendererClass=&NiagaraSpriteRendererProperties, rendererIndex=2}
  ]
 }
 Goop {
  emitterUpdateScript [
   &EmitterState(Life Cycle Mode=::NewEnumerator1, Loop Behavior=::NewEnumerator1, Loop Duration=2, UseLoopDelay=false)
   &SpawnBurst_Instantaneous(Use Spawn Probability=false, UseLoopCountLimit=false)
   &SpawnBurst_Instantaneous(Spawn Group=1, Use Spawn Probability=false, UseLoopCountLimit=false) {module=SpawnBurst_Instantaneous001}
   &SpawnBurst_Instantaneous(
    Spawn Count=&RandomRangeInt(Minimum=25, Maximum=30)
    Spawn Group=2
    Use Spawn Probability=false
    UseLoopCountLimit=false
   ) {enabled=false, module=SpawnBurst_Instantaneous002}
  ]
  particleSpawnScript [
   /SetVariables(Particles.VisibilityTag=&SpawnGroup()) {isSetParameters=true, module=SetVariables_#2}
   &InitializeParticle(
    Lifetime=&MergeHelper_FloatSwitch(
     0=0.4
     1=0.5
     2=&RandomRangeFloat(Minimum=0.4, Maximum=0.7)
     3=?
     Index=&SpawnGroup()
     Num=::NewEnumerator1
    )
    Sprite Rotation Mode=::NewEnumerator2
    Sprite Size Mode=::NewEnumerator3
    Uniform Sprite Size=&Multiply_Float(
     A=&MergeHelper_FloatSwitch(
      0=75
      1=95
      2=&RandomRangeFloat(Minimum=6, Maximum=8)
      3=?
      Index=&SpawnGroup()
      Num=::NewEnumerator1
     )
     B=&MakeCustomFloatFromBool(Bool=$User.IsLargeBody, True Float=1.1, False Float=1)
    )
    UsePositionOffset=false
    Write Material Random=true
   )
   &ShapeLocation(
    Apply To Particle Position=::NewEnumerator1
    Hemisphere Distribution=&MergeHelper_Vector2Switch(0=<0 0>, 1=<0 0>, 2=<1 1>, 3=?, 4=?, Index=&SpawnGroup(), Num=::NewEnumerator1)
    MaskSpawnGroup=false
    Override Seed=false
    Shape Origin=&VectorFromFloat(Value=0)
    Sphere Radius=&MergeHelper_FloatSwitch(0=0, 1=0, 2=15, 3=?, Index=&SpawnGroup(), Num=::NewEnumerator1)
    Sphere Surface Distribution=0.4524
    UseOwnerScale=true
   )
   &SkeletalMeshLocation(
    Bone Sampling Mode=::NewEnumerator4
    MaskSpawnGroup=false
    Orientation Sampling=::NewEnumerator1
    Position Sampling=::NewEnumerator1
    Sampled Position Offset=&AddVectorToPosition(Position=$Particles.ShapeLocation.ShapePosition, Vector=<15 0 0>)
    Skeletal Mesh=$System.Skeletal Mesh
   )
   &AddVelocity(
    Module.Use Falloff Distance=false
    Origin Offset=&MergeHelper_VectorSwitch(0=<0 0 0>, 1=<0 0 0>, 2=<0 0 5>, 3=?, Index=&SpawnGroup(), Num=::NewEnumerator1)
    Override Seed=false
    Velocity Mode=::NewEnumerator1
    Velocity Origin=$Output.SkeletalMeshLocation.SampledPositionWithoutOffset
    Velocity Speed=&MergeHelper_FloatSwitch(
     0=50
     1=25
     2=&RandomRangeFloat(Minimum=100, Maximum=300)
     3=?
     Index=&SpawnGroup()
     Num=::NewEnumerator1
    )
   ) {module=AddVelocity001}
   &CameraOffset(Camera Offset Amount=-25)
   /SetVariables(Particles.Frame:NiagaraFloat=&MergeHelper_FloatSwitch(
     0=0
     1=&MakeFloatFromInt(Int=&RandomRangeInt(Minimum=0, Maximum=3))
     2=&MakeFloatFromInt(Int=&RandomRangeInt(Minimum=0, Maximum=3))
     3=?
     Index=&SpawnGroup()
     Num=::NewEnumerator1
    )) {isSetParameters=true, module=SetVariables_#3}
  ]
  particleUpdateScript [
   &ParticleState()
   &ScaleSpriteSize(
    Scale Sprite Size Mode=::NewEnumerator0
    Uniform Scale Factor=&MergeHelper_FloatSwitch(
     0=&FloatFromCurve(FloatCurve=DataInterface() {propertyValues {Curve {keys [
          {interpMode=RCIM_Cubic, leaveTangent=2.7083, tangentMode=RCTM_Break, time=0, value=0}
          {arriveTangent=1.1018, interpMode=RCIM_Cubic, leaveTangent=1.1018, time=0.4356, value=0.8}
          {interpMode=RCIM_Cubic, leaveTangent=0.9481, time=0.9983, value=1.1}
         ]}}})
     1=&FloatFromCurve(FloatCurve=DataInterface() {propertyValues {Curve {keys [
          {interpMode=RCIM_Cubic, leaveTangent=2.7083, tangentMode=RCTM_Break, time=0, value=0}
          {arriveTangent=1.524, interpMode=RCIM_Cubic, leaveTangent=1.524, time=0.4356, value=0.8}
          {arriveTangent=0.5331, interpMode=RCIM_Cubic, leaveTangent=0.5331, time=0.647, value=0.986}
          {interpMode=RCIM_Cubic, leaveTangent=0.9481, time=0.9983, value=1.1}
         ]}}})
     2=&FloatFromCurve(FloatCurve=DataInterface() {propertyValues {Curve {
         defaultValue=1
         keys [
          {arriveTangent=4.7388, interpMode=RCIM_Cubic, leaveTangent=4.7388, tangentMode=RCTM_User, time=0, value=0}
          {arriveTangentWeight=0.1274, interpMode=RCIM_Cubic, leaveTangentWeight=0.7778, tangentMode=RCTM_User, time=0.2, value=1}
          {arriveTangent=-1.9979, interpMode=RCIM_Cubic, leaveTangent=-1.9979, tangentMode=RCTM_User, time=1, value=0}
         ]
        }}})
     3=?
     Index=&SpawnGroup()
     Num=::NewEnumerator1
    )
   )
   &GravityForce(Gravity=&MergeHelper_VectorSwitch(
     0=<0 0 -200>
     1=<0 0 -200>
     2=&RandomRangeVector(Minimum=<0 0 -350>, Maximum=<0 0 -500>)
     3=?
     Index=&SpawnGroup()
     Num=::NewEnumerator1
    ))
   &DynamicMaterialParameters(
    Index 0 Param 1=&MergeHelper_FloatSwitch(
     0=&RandomRangeFloat(Minimum=0, Maximum=1)
     1=0
     2=0
     3=?
     Index=&SpawnGroup()
     Num=::NewEnumerator1
    )
    Index 0 Param 3=&FloatFromCurve(FloatCurve=DataInterface() {propertyValues {Curve {keys [
         {interpMode=RCIM_Cubic, tangentMode=RCTM_Break, time=0.2, value=-0.6}
         {arriveTangent=-0.3316, interpMode=RCIM_Cubic, tangentMode=RCTM_Break, time=1, value=-1.5}
        ]}}})
    Index 0 Param 4=&FloatFromCurve(FloatCurve=DataInterface() {propertyValues {Curve {keys [
         {arriveTangent=2.6651, interpMode=RCIM_Cubic, leaveTangent=2.6651, tangentMode=RCTM_User, time=0, value=0.1}
         {arriveTangent=3.4594, interpMode=RCIM_Cubic, leaveTangent=3.4594, tangentMode=RCTM_User, time=1, value=3}
        ]}}})
    Module.Index0Float0Write=true
    Module.Index0Float1Write=true
    Module.Index0Float2Write=true
    Module.Index0Float3Write=true
   ) {module=DynamicMaterialParameters001}
   &Drag(Use Linear Drag=true, Use Rotational Drag=true)
   &SolveForcesAndVelocity(Clamp Velocity=false, Limit Acceleration=false)
   &ScaleSpriteSizeBySpeed(
    Max Scale Factor=&MergeHelper_Vector2Switch(0=<1 1>, 1=<1 1>, 2=<0.5 3>, 3=?, 4=?, Index=&SpawnGroup(), Num=::NewEnumerator1)
    Min Scale Factor=&MergeHelper_Vector2Switch(0=<1 1>, 1=<1 1>, 2=<0.9 1.2>, 3=?, 4=?, Index=&SpawnGroup(), Num=::NewEnumerator1)
   )
  ]
  rendererClasses [&NiagaraSpriteRendererProperties]
  renderers [
   {rendererClass=&NiagaraSpriteRendererProperties}
   {rendererClass=&NiagaraSpriteRendererProperties, rendererIndex=1}
   {rendererClass=&NiagaraSpriteRendererProperties, rendererIndex=2}
  ]
 }
}
summary {
 systemName=NS_Aura_ShadedExpanse_Start
 emitters [
  {emitterName=CenterFlare, rendererClasses [&NiagaraSpriteRendererProperties]}
  {emitterName=Corona_Glow_Combined, rendererClasses [&NiagaraSpriteRendererProperties]}
  {emitterName=Goop, rendererClasses [&NiagaraSpriteRendererProperties]}
  {emitterName=Drops, rendererClasses [&NiagaraSpriteRendererProperties]}
 ]
 userVariables [
  User.bIsFrontEnd {type=NiagaraBool}
  User.bIsFrontEndPreview {type=NiagaraBool}
  User.IsLargeBody {type=NiagaraBool}
 ]
}
